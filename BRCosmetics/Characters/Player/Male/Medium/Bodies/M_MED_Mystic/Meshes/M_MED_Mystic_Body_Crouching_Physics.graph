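# Physics asset for the crouching Mystic medium male body, covering its dangling belt, cloth belt and left and right lapels.
# Capsule bodies on the dyn_ bones form short chains with narrow swing limits and mostly zero twist, allowing twist only on the front lapels.
# Small MSTR root bodies attach through free-rotation constraints, the belts to the pelvis and the lapels to their thighs, which have larger capsules.
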PhysicsAsset /BRCosmetics/Characters/Player/Male/Medium/Bodies/M_MED_Mystic/Meshes/M_MED_Mystic_Body_Crouching_Physics
bodies {
 dyn_belt_01 [{shapeType=Capsule, length=2.9448, radius=1.7209, shapeName=dyn_belt_01_capsule, center=<2.4793 0 0>, rotation=<rot 90 0 0>}]
 dyn_belt_02 [{shapeType=Capsule, length=5.0324, radius=1.7209, shapeName=dyn_belt_02_capsule, center=<4.2343 0 0>, rotation=<rot 90 0 0>}]
 dyn_belt_03 [{shapeType=Capsule, length=3, radius=1, shapeName=dyn_belt_03_capsule, center=<1.9135 0 0>, rotation=<rot 90 0 0>}]
 dyn_belt_05 [{shapeType=Capsule, length=5.5, radius=5, shapeName=dyn_belt_05_capsule, center=<3.0913 0 3.4535>, rotation=<rot 84.7915 0 0>}]
 dyn_belt_MSTR [{shapeType=Capsule, length=0.505, radius=0.505, shapeName=dyn_belt_MSTR_capsule, rotation=<rot 0 -90 0>}]
 dyn_clothBelt_01 [{shapeType=Capsule, length=1, radius=1, shapeName=dyn_clothBelt_01_capsule}]
 dyn_clothBelt_02 [{shapeType=Capsule, length=1, radius=1, shapeName=dyn_clothBelt_02_capsule}]
 dyn_clothBelt_03 [{shapeType=Capsule, length=7, radius=2.75, shapeName=dyn_clothBelt_03_capsule, center=<5.674 0 1.26>, rotation=<rot 0 0 90>}]
 dyn_clothBelt_MSTR [{shapeType=Capsule, length=0.505, radius=0.505, shapeName=dyn_clothBelt_MSTR_capsule, rotation=<rot 0 -90 0>}]
 dyn_lapel_front_l_01 [{shapeType=Capsule, length=5.545, radius=1.5172, shapeName=dyn_lapel_front_l_01_capsule, center=<6.7998 0 0>}]
 dyn_lapel_front_l_02 [{
   shapeType=Capsule
   length=5.494
   radius=2.0128
   shapeName=dyn_lapel_front_l_02_capsule
   center=<6.9083 0.7171 3.8338>
   rotation=<rot -8.3997 -0.0001 10.4835>
  }]
 dyn_lapel_front_l_MSTR [{shapeType=Capsule, length=0.505, radius=0.505, shapeName=dyn_lapel_front_l_MSTR_capsule, rotation=<rot 0 -90 0>}]
 dyn_lapel_sideA_l_01 [{shapeType=Capsule, length=7.5655, radius=1.6186, shapeName=dyn_lapel_sideA_l_01_capsule, center=<4.3638 0 0>}]
 dyn_lapel_sideA_l_02 [{shapeType=Capsule, length=5.304, radius=2.3944, shapeName=dyn_lapel_sideA_l_02_capsule, center=<4.3638 0 0>}]
 dyn_lapel_sideB_l_01 [{shapeType=Capsule, length=11.645, radius=1.6186, shapeName=dyn_lapel_sideB_l_01_capsule, center=<4.3638 0 0>}]
 dyn_lapel_sideB_l_02 [{shapeType=Capsule, length=5.304, radius=2.3944, shapeName=dyn_lapel_sideB_l_02_capsule, center=<4.3638 0 0>}]
 dyn_lapel_side_l_MSTR [{shapeType=Capsule, length=0.505, radius=0.505, shapeName=dyn_lapel_side_l_MSTR_capsule, rotation=<rot 0 -90 0>}]
 dyn_rapel_front_r_01 [{shapeType=Capsule, length=4.6268, radius=2.4354, shapeName=dyn_rapel_front_r_01_capsule, center=<4.3638 0 0>}]
 dyn_rapel_front_r_02 [{shapeType=Capsule, length=4.6849, radius=2.9923, shapeName=dyn_rapel_front_r_02_capsule, center=<4.3638 0 0>}]
 dyn_rapel_front_r_MSTR [{shapeType=Capsule, length=0.505, radius=0.505, shapeName=dyn_rapel_front_r_MSTR_capsule, rotation=<rot 0 -90 0>}]
 dyn_rapel_sideA_r_01 [{shapeType=Capsule, length=11.645, radius=1.6186, shapeName=dyn_rapel_sideA_r_01_capsule, center=<4.3638 0 0>}]
 dyn_rapel_sideA_r_02 [{shapeType=Capsule, length=8.5655, radius=1.7135, shapeName=dyn_rapel_sideA_r_02_capsule, center=<4.3638 0 0>}]
 dyn_rapel_sideB_r_01 [{shapeType=Capsule, length=11.645, radius=1.6186, shapeName=dyn_rapel_sideB_r_01_capsule, center=<4.3638 0 0>}]
 dyn_rapel_sideB_r_02 [{shapeType=Capsule, length=5.6076, radius=2.0909, shapeName=dyn_rapel_sideB_r_02_capsule, center=<4.3638 0 0>}]
 dyn_rapel_side_r_MSTR [{shapeType=Capsule, length=0.505, radius=0.505, shapeName=dyn_rapel_side_r_MSTR_capsule, rotation=<rot 0 -90 0>}]
 pelvis [{shapeType=Capsule, length=12.0357, radius=9.2671, shapeName=pelvis_capsule}]
 thigh_l [{shapeType=Capsule, length=27.8102, radius=7.8179, shapeName=thigh_l_capsule, center=<-14.1757 0 2.1261>, rotation=<rot 90 0 0>}]
 thigh_r [{shapeType=Capsule, length=25.2396, radius=9.1568, shapeName=thigh_r_capsule, center=<14.5562 0 -1.7787>, rotation=<rot -90 0 0>}]
}
constraints [
 {bone1Name=dyn_belt_02, bone2Name=dyn_belt_01, swing1LimitDegrees=20, swing2LimitDegrees=20, twistLimitDegrees=0}
 {bone1Name=dyn_belt_03, bone2Name=dyn_belt_02, swing1LimitDegrees=20, swing2LimitDegrees=20, twistLimitDegrees=0}
 {bone1Name=dyn_belt_05, bone2Name=dyn_belt_03, swing1LimitDegrees=20, swing2LimitDegrees=20, twistLimitDegrees=0}
 {bone1Name=dyn_clothBelt_02, bone2Name=dyn_clothBelt_01, swing1LimitDegrees=20, swing2LimitDegrees=20, twistLimitDegrees=0}
 {bone1Name=dyn_clothBelt_03, bone2Name=dyn_clothBelt_02, swing1LimitDegrees=20, swing2LimitDegrees=20, twistLimitDegrees=0}
 {bone1Name=dyn_belt_01, bone2Name=dyn_belt_MSTR, swing1LimitDegrees=10, swing2LimitDegrees=20, twistLimitDegrees=0}
 {bone1Name=dyn_clothBelt_01, bone2Name=dyn_clothBelt_MSTR, swing1LimitDegrees=10, swing2LimitDegrees=20, twistLimitDegrees=0}
 {
  bone1Name=thigh_l
  bone2Name=pelvis
  swing1LimitDegrees=45
  swing1Motion=Free
  swing2LimitDegrees=45
  swing2Motion=Free
  twistLimitDegrees=45
  twistMotion=Free
 }
 {
  bone1Name=thigh_r
  bone2Name=pelvis
  swing1LimitDegrees=45
  swing1Motion=Free
  swing2LimitDegrees=45
  swing2Motion=Free
  twistLimitDegrees=45
  twistMotion=Free
 }
 {
  bone1Name=dyn_clothBelt_MSTR
  bone2Name=pelvis
  swing1LimitDegrees=45
  swing1Motion=Free
  swing2LimitDegrees=45
  swing2Motion=Free
  twistLimitDegrees=45
  twistMotion=Free
 }
 {
  bone1Name=dyn_belt_MSTR
  bone2Name=pelvis
  swing1LimitDegrees=45
  swing1Motion=Free
  swing2LimitDegrees=45
  swing2Motion=Free
  twistLimitDegrees=45
  twistMotion=Free
 }
 {bone1Name=dyn_lapel_front_l_01, bone2Name=dyn_lapel_front_l_MSTR, swing1LimitDegrees=20, swing2LimitDegrees=20, twistLimitDegrees=20}
 {bone1Name=dyn_lapel_front_l_02, bone2Name=dyn_lapel_front_l_01, swing1LimitDegrees=20, swing2LimitDegrees=20, twistLimitDegrees=30}
 {bone1Name=dyn_lapel_sideA_l_01, bone2Name=dyn_lapel_side_l_MSTR, swing1LimitDegrees=15, swing2LimitDegrees=15, twistLimitDegrees=0}
 {bone1Name=dyn_lapel_sideA_l_02, bone2Name=dyn_lapel_sideA_l_01, swing1LimitDegrees=15, swing2LimitDegrees=8, twistLimitDegrees=0}
 {bone1Name=dyn_lapel_sideB_l_01, bone2Name=dyn_lapel_side_l_MSTR, swing1LimitDegrees=15, swing2LimitDegrees=15, twistLimitDegrees=0}
 {bone1Name=dyn_lapel_sideB_l_02, bone2Name=dyn_lapel_sideB_l_01, swing1LimitDegrees=15, swing2LimitDegrees=15, twistLimitDegrees=0}
 {bone1Name=dyn_rapel_front_r_01, bone2Name=dyn_rapel_front_r_MSTR, swing1LimitDegrees=20, swing2LimitDegrees=20, twistLimitDegrees=20}
 {bone1Name=dyn_rapel_front_r_02, bone2Name=dyn_rapel_front_r_01, swing1LimitDegrees=20, swing2LimitDegrees=20, twistLimitDegrees=20}
 {bone1Name=dyn_rapel_sideA_r_01, bone2Name=dyn_rapel_side_r_MSTR, swing1LimitDegrees=15, swing2LimitDegrees=15, twistLimitDegrees=0}
 {bone1Name=dyn_rapel_sideA_r_02, bone2Name=dyn_rapel_sideA_r_01, swing1LimitDegrees=15, swing2LimitDegrees=15, twistLimitDegrees=0}
 {bone1Name=dyn_rapel_sideB_r_01, bone2Name=dyn_rapel_side_r_MSTR, swing1LimitDegrees=15, swing2LimitDegrees=15, twistLimitDegrees=0}
 {bone1Name=dyn_rapel_sideB_r_02, bone2Name=dyn_rapel_sideB_r_01, swing1LimitDegrees=15, swing2LimitDegrees=15, twistLimitDegrees=0}
 {
  bone1Name=dyn_lapel_front_l_MSTR
  bone2Name=thigh_l
  swing1LimitDegrees=45
  swing1Motion=Free
  swing2LimitDegrees=45
  swing2Motion=Free
  twistLimitDegrees=45
  twistMotion=Free
 }
 {
  bone1Name=dyn_lapel_side_l_MSTR
  bone2Name=thigh_l
  swing1LimitDegrees=45
  swing1Motion=Free
  swing2LimitDegrees=45
  swing2Motion=Free
  twistLimitDegrees=45
  twistMotion=Free
 }
 {
  bone1Name=dyn_rapel_front_r_MSTR
  bone2Name=thigh_r
  swing1LimitDegrees=45
  swing1Motion=Free
  swing2LimitDegrees=45
  swing2Motion=Free
  twistLimitDegrees=45
  twistMotion=Free
 }
 {
  bone1Name=dyn_rapel_side_r_MSTR
  bone2Name=thigh_r
  swing1LimitDegrees=45
  swing1Motion=Free
  swing2LimitDegrees=45
  swing2Motion=Free
  twistLimitDegrees=45
  twistMotion=Free
 }
]
