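# Physics asset for the male medium Mystic cosmetic body, covering secondary motion of its belt, cloth belt and left/right lapels.
# Capsules on the pelvis and thighs form the body core, with tiny MSTR anchor bodies for each dangling group.
# Belt, cloth belt and ring anchors constrain freely to the pelvis, lapel anchors to their thighs.
# Chains below the anchors use narrow swing limits and mostly zero twist limits.
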PhysicsAsset /BRCosmetics/Characters/Player/Male/Medium/Bodies/M_MED_Mystic/Meshes/M_MED_Mystic_Body_Physics
bodies {
 dyn_belt_01 [{shapeType=Capsule, length=4.0035, radius=1.2836, shapeName=dyn_belt_01_capsule, center=<2.8859 0 0>, rotation=<rot 90 0 0>}]
 dyn_belt_02 [{shapeType=Capsule, length=6.4291, radius=1.3959, shapeName=dyn_belt_02_capsule, center=<4.2343 0 0>, rotation=<rot 90 0 0>}]
 dyn_belt_03 [{shapeType=Capsule, length=5.797, radius=1.7981, shapeName=dyn_belt_03_capsule, center=<3.8422 0 0>, rotation=<rot 90 0 0>}]
 dyn_belt_05 [{shapeType=Capsule, length=5.25, radius=4.25, shapeName=dyn_belt_05_capsule, center=<4.4067 0 2.8783>, rotation=<rot -87.0578 180 -179.9999>}]
 dyn_belt_MSTR [{shapeType=Capsule, length=0.505, radius=0.505, shapeName=dyn_belt_MSTR_capsule, rotation=<rot 0 -90 0>}]
 dyn_clothBelt_01 [{
   shapeType=Capsule
   length=14.1512
   radius=5
   shapeName=dyn_clothBelt_01_capsule
   center=<6.7147 3.0237 -4.869>
   rotation=<rot 0.0001 0 104.1973>
  }]
 dyn_clothBelt_02 [
  {
   shapeType=Capsule
   length=9.4012
   radius=3.75
   shapeName=dyn_clothBelt_02_capsule
   center=<6.7147 0.1353 -4.0862>
   rotation=<rot 0.0001 0.0001 101.8808>
  }
  {
   shapeType=Capsule
   length=13.7101
   radius=5.0289
   shapeName=dyn_clothBelt_02_capsule
   center=<9.5717 0.8451 -4.0568>
   rotation=<rot 0 90.0001 85.5256>
  }
 ]
 dyn_clothBelt_03 [{shapeType=Capsule, length=6.4603, radius=3.2391, shapeName=dyn_clothBelt_03_capsule, center=<4.7 0 -3.2>, rotation=<rot 0 0 90>}]
 dyn_clothBelt_MSTR [{shapeType=Capsule, length=0.505, radius=0.505, shapeName=dyn_clothBelt_MSTR_capsule, rotation=<rot 0 -90 0>}]
 dyn_lapel_front_l_01 [{shapeType=Capsule, length=5.545, radius=1.5172, shapeName=dyn_lapel_front_l_01_capsule, center=<4.3638 0 0>}]
 dyn_lapel_front_l_02 [{
   shapeType=Capsule
   length=6.1929
   radius=2.4324
   shapeName=dyn_lapel_front_l_02_capsule
   center=<6.7149 0 1.5475>
   rotation=<rot -8.2934 -0.0001 0>
  }]
 dyn_lapel_front_l_MSTR [{shapeType=Capsule, length=0.505, radius=0.505, shapeName=dyn_lapel_front_l_MSTR_capsule, rotation=<rot 0 -90 0>}]
 dyn_lapel_sideA_l_01 [{shapeType=Capsule, length=7.5655, radius=1.6186, shapeName=dyn_lapel_sideA_l_01_capsule, center=<4.3638 0 0>}]
 dyn_lapel_sideA_l_02 [{shapeType=Capsule, length=5.304, radius=2.3944, shapeName=dyn_lapel_sideA_l_02_capsule, center=<4.3638 0 0>}]
 dyn_lapel_sideB_l_01 [{shapeType=Capsule, length=11.645, radius=1.6186, shapeName=dyn_lapel_sideB_l_01_capsule, center=<4.3638 0 0>}]
 dyn_lapel_sideB_l_02 [{shapeType=Capsule, length=5.304, radius=2.3944, shapeName=dyn_lapel_sideB_l_02_capsule, center=<4.3638 0 0>}]
 dyn_lapel_side_l_MSTR [{shapeType=Capsule, length=0.505, radius=0.505, shapeName=dyn_lapel_side_l_MSTR_capsule, rotation=<rot 0 -90 0>}]
 dyn_rapel_front_r_01 [{shapeType=Capsule, length=4.6268, radius=2.4354, shapeName=dyn_rapel_front_r_01_capsule, center=<4.3638 0 0>}]
 dyn_rapel_front_r_02 [{shapeType=Capsule, length=4.6849, radius=2.9923, shapeName=dyn_rapel_front_r_02_capsule, center=<4.3638 0 0>}]
 dyn_rapel_front_r_MSTR [{shapeType=Capsule, length=0.505, radius=0.505, shapeName=dyn_rapel_front_r_MSTR_capsule, rotation=<rot 0 -90 0>}]
 dyn_rapel_sideA_r_01 [{shapeType=Capsule, length=11.645, radius=1.6186, shapeName=dyn_rapel_sideA_r_01_capsule, center=<4.3638 0 0>}]
 dyn_rapel_sideA_r_02 [{shapeType=Capsule, length=8.5655, radius=1.7135, shapeName=dyn_rapel_sideA_r_02_capsule, center=<4.3638 0 0>}]
 dyn_rapel_sideB_r_01 [{shapeType=Capsule, length=11.395, radius=1.8686, shapeName=dyn_rapel_sideB_r_01_capsule, center=<4.3638 0 0>}]
 dyn_rapel_sideB_r_02 [{shapeType=Capsule, length=5.3576, radius=2.3409, shapeName=dyn_rapel_sideB_r_02_capsule, center=<4.3638 0 0>}]
 dyn_rapel_side_r_MSTR [{shapeType=Capsule, length=0.505, radius=0.505, shapeName=dyn_rapel_side_r_MSTR_capsule, rotation=<rot 0 -90 0>}]
 dyn_ring_MSTR [{shapeType=Capsule, length=0.505, radius=0.505, shapeName=dyn_ring_MSTR_capsule, rotation=<rot 0 -90 0>}]
 pelvis [{shapeType=Capsule, length=12.0357, radius=9.2671, shapeName=pelvis_capsule}]
 thigh_l [
  {shapeType=Capsule, length=26.1739, radius=9.4542, shapeName=thigh_l_capsule, center=<-14.1757 0 2.1261>, rotation=<rot 90 0 0>}
  {
   shapeType=Capsule
   length=10.1927
   radius=9.5471
   shapeName=thigh_l_capsule
   center=<-0.5178 2.1532 3.184>
   rotation=<rot 85.9157 166.6549 -178.8084>
  }
 ]
 thigh_r [{shapeType=Capsule, length=25.2396, radius=9.1568, shapeName=thigh_r_capsule, center=<14.5562 0 -1.7787>, rotation=<rot -90 0 0>}]
}
constraints [
 {bone1Name=dyn_belt_02, bone2Name=dyn_belt_01, swing1LimitDegrees=0, swing2LimitDegrees=20, twistLimitDegrees=0}
 {bone1Name=dyn_belt_03, bone2Name=dyn_belt_02, swing1LimitDegrees=20, swing2LimitDegrees=20, twistLimitDegrees=0}
 {bone1Name=dyn_belt_05, bone2Name=dyn_belt_03, swing1LimitDegrees=20, swing2LimitDegrees=20, twistLimitDegrees=0}
 {bone1Name=dyn_clothBelt_02, bone2Name=dyn_clothBelt_01, swing1LimitDegrees=20, swing2LimitDegrees=20, twistLimitDegrees=0}
 {bone1Name=dyn_clothBelt_03, bone2Name=dyn_clothBelt_02, swing1LimitDegrees=10, swing2LimitDegrees=10, twistLimitDegrees=0}
 {bone1Name=dyn_belt_01, bone2Name=dyn_belt_MSTR, swing1LimitDegrees=0, swing2LimitDegrees=8, twistLimitDegrees=0}
 {bone1Name=dyn_clothBelt_01, bone2Name=dyn_clothBelt_MSTR, swing1LimitDegrees=10, swing2LimitDegrees=20, twistLimitDegrees=0}
 {
  bone1Name=thigh_l
  bone2Name=pelvis
  swing1LimitDegrees=45
  swing1Motion=Free
  swing2LimitDegrees=45
  swing2Motion=Free
  twistLimitDegrees=45
  twistMotion=Free
 }
 {
  bone1Name=thigh_r
  bone2Name=pelvis
  swing1LimitDegrees=45
  swing1Motion=Free
  swing2LimitDegrees=45
  swing2Motion=Free
  twistLimitDegrees=45
  twistMotion=Free
 }
 {
  bone1Name=dyn_ring_MSTR
  bone2Name=pelvis
  swing1LimitDegrees=45
  swing1Motion=Free
  swing2LimitDegrees=45
  swing2Motion=Free
  twistLimitDegrees=45
  twistMotion=Free
 }
 {
  bone1Name=dyn_clothBelt_MSTR
  bone2Name=pelvis
  swing1LimitDegrees=45
  swing1Motion=Free
  swing2LimitDegrees=45
  swing2Motion=Free
  twistLimitDegrees=45
  twistMotion=Free
 }
 {
  bone1Name=dyn_belt_MSTR
  bone2Name=pelvis
  swing1LimitDegrees=45
  swing1Motion=Free
  swing2LimitDegrees=45
  swing2Motion=Free
  twistLimitDegrees=45
  twistMotion=Free
 }
 {bone1Name=dyn_lapel_front_l_01, bone2Name=dyn_lapel_front_l_MSTR, swing1LimitDegrees=20, swing2LimitDegrees=20, twistLimitDegrees=20}
 {bone1Name=dyn_lapel_front_l_02, bone2Name=dyn_lapel_front_l_01, swing1LimitDegrees=20, swing2LimitDegrees=20, twistLimitDegrees=10}
 {bone1Name=dyn_lapel_sideA_l_01, bone2Name=dyn_lapel_side_l_MSTR, swing1LimitDegrees=15, swing2LimitDegrees=15, twistLimitDegrees=0}
 {bone1Name=dyn_lapel_sideA_l_02, bone2Name=dyn_lapel_sideA_l_01, swing1LimitDegrees=15, swing2LimitDegrees=15, twistLimitDegrees=0}
 {bone1Name=dyn_lapel_sideB_l_01, bone2Name=dyn_lapel_side_l_MSTR, swing1LimitDegrees=15, swing2LimitDegrees=15, twistLimitDegrees=0}
 {bone1Name=dyn_lapel_sideB_l_02, bone2Name=dyn_lapel_sideB_l_01, swing1LimitDegrees=15, swing2LimitDegrees=15, twistLimitDegrees=0}
 {bone1Name=dyn_rapel_front_r_01, bone2Name=dyn_rapel_front_r_MSTR, swing1LimitDegrees=20, swing2LimitDegrees=20, twistLimitDegrees=20}
 {bone1Name=dyn_rapel_front_r_02, bone2Name=dyn_rapel_front_r_01, swing1LimitDegrees=20, swing2LimitDegrees=20, twistLimitDegrees=20}
 {bone1Name=dyn_rapel_sideA_r_01, bone2Name=dyn_rapel_side_r_MSTR, swing1LimitDegrees=15, swing2LimitDegrees=15, twistLimitDegrees=0}
 {bone1Name=dyn_rapel_sideA_r_02, bone2Name=dyn_rapel_sideA_r_01, swing1LimitDegrees=15, swing2LimitDegrees=15, twistLimitDegrees=0}
 {bone1Name=dyn_rapel_sideB_r_01, bone2Name=dyn_rapel_side_r_MSTR, swing1LimitDegrees=15, swing2LimitDegrees=15, twistLimitDegrees=0}
 {bone1Name=dyn_rapel_sideB_r_02, bone2Name=dyn_rapel_sideB_r_01, swing1LimitDegrees=15, swing2LimitDegrees=15, twistLimitDegrees=0}
 {
  bone1Name=dyn_lapel_front_l_MSTR
  bone2Name=thigh_l
  swing1LimitDegrees=45
  swing1Motion=Free
  swing2LimitDegrees=45
  swing2Motion=Free
  twistLimitDegrees=45
  twistMotion=Free
 }
 {
  bone1Name=dyn_lapel_side_l_MSTR
  bone2Name=thigh_l
  swing1LimitDegrees=45
  swing1Motion=Free
  swing2LimitDegrees=45
  swing2Motion=Free
  twistLimitDegrees=45
  twistMotion=Free
 }
 {
  bone1Name=dyn_rapel_front_r_MSTR
  bone2Name=thigh_r
  swing1LimitDegrees=45
  swing1Motion=Free
  swing2LimitDegrees=45
  swing2Motion=Free
  twistLimitDegrees=45
  twistMotion=Free
 }
 {
  bone1Name=dyn_rapel_side_r_MSTR
  bone2Name=thigh_r
  swing1LimitDegrees=45
  swing1Motion=Free
  swing2LimitDegrees=45
  swing2Motion=Free
  twistLimitDegrees=45
  twistMotion=Free
 }
]
