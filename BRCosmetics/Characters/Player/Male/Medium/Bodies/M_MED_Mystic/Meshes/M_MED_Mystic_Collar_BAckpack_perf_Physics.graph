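# Physics asset for the Mystic medium male body's collar/backpack mesh, defining collision and constraints for a four-chain cape.
# Outer chains hang from spine_05 through clavicle cape bones and middle chains from backpack_root, with per-joint swing and twist limits.
# Free-angular constraints cross-link neighboring chains at the second and fourth segments.
# Capsules cover torso, limbs and backpack, while dyn_root_slope adds two spheres, one huge and offset like a ground surface.
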PhysicsAsset /BRCosmetics/Characters/Player/Male/Medium/Bodies/M_MED_Mystic/Meshes/M_MED_Mystic_Collar_BAckpack_perf_Physics
bodies {
 backpack_root [{shapeType=Capsule, length=1.5, radius=7.005, shapeName=backpack_root_capsule, center=<0 -27.034 -5.6066>, rotation=<rot 0 -90 0>}]
 calf_l [{shapeType=Capsule, length=30.0334, radius=8.8128, shapeName=calf_l_capsule, center=<-21.7328 1.7569 0>, rotation=<rot 90 0 13.1977>}]
 calf_r [{shapeType=Capsule, length=30.0334, radius=8.8128, shapeName=calf_r_capsule, center=<21.7328 -1.7569 0>, rotation=<rot -90 180 166.8032>}]
 cape_01_l [{shapeType=Capsule, length=18.75, radius=9.25, shapeName=cape_01_l_capsule, center=<7.0671 -2 5>, rotation=<rot 80 89.9994 179.9994>}]
 cape_01_r [{shapeType=Capsule, length=9.7362, radius=9.2781, shapeName=cape_01_r_capsule, center=<8.7336 0 5>, rotation=<rot 0 0 85>}]
 cape_02_l [{shapeType=Capsule, length=6.25, radius=16.5, shapeName=cape_02_l_capsule, center=<12.8625 0 9>, rotation=<rot 90 0 0>}]
 cape_02_r [{shapeType=Capsule, length=6, radius=16, shapeName=cape_02_r_capsule, center=<12.8625 0 9>, rotation=<rot 90 0 0>}]
 cape_03_l [{shapeType=Capsule, length=10.941, radius=12.255, shapeName=cape_03_l_capsule, center=<11.9516 0 4>, rotation=<rot 0 0 105>}]
 cape_03_r [{shapeType=Capsule, length=14.691, radius=12.255, shapeName=cape_03_r_capsule, center=<11.9516 0 4>, rotation=<rot -0.0009 0 75>}]
 cape_04_l [{shapeType=Capsule, length=13.7939, radius=10.6132, shapeName=cape_04_l_capsule, center=<10 0 4>, rotation=<rot -0.0002 0 105>}]
 cape_04_r [{shapeType=Capsule, length=15.5, radius=10.5, shapeName=cape_04_r_capsule, center=<10 0 4>, rotation=<rot -0.0003 0 75>}]
 cape_05_l [{shapeType=Capsule, length=16.0389, radius=7.9839, shapeName=cape_05_l_capsule, center=<6.8829 0 3>, rotation=<rot -0.0001 0 100>}]
 cape_05_r [{shapeType=Capsule, length=16.5389, radius=7.9839, shapeName=cape_05_r_capsule, center=<6.8829 0 3>, rotation=<rot -0.0002 0 80>}]
 cape_l_MSTR [{shapeType=Capsule, length=1, radius=1, shapeName=cape_l_MSTR_capsule, rotation=<rot 90 0 0>}]
 cape_mid_01_l [{shapeType=Capsule, length=8.8444, radius=9.7903, shapeName=cape_mid_01_r_capsule, center=<7 0 8>, rotation=<rot 0 0 90>}]
 cape_mid_01_r [{shapeType=Capsule, length=7.3444, radius=11.7903, shapeName=cape_mid_01_r_capsule, center=<7.4618 0 8>, rotation=<rot -0.0001 0 85>}]
 cape_mid_02_l [{shapeType=Capsule, length=5.865, radius=15.2492, shapeName=cape_mid_02_r_capsule, center=<14.1479 0 8.5>, rotation=<rot -90 0 0>}]
 cape_mid_02_r [{shapeType=Capsule, length=5.865, radius=14.9992, shapeName=cape_mid_02_r_capsule, center=<13.0851 0 8.5>, rotation=<rot 90 0 0>}]
 cape_mid_03_l [{shapeType=Capsule, length=16.941, radius=13.505, shapeName=cape_mid_03_r_capsule, center=<11.9516 0 7>, rotation=<rot -0.0001 0 95>}]
 cape_mid_03_r [{shapeType=Capsule, length=17.191, radius=13.505, shapeName=cape_mid_03_r_capsule, center=<11.9516 0 7>, rotation=<rot -0.0001 0 85>}]
 cape_mid_04_l [{shapeType=Capsule, length=17.9946, radius=10.1132, shapeName=cape_mid_04_r_capsule, center=<10.2079 0 3>, rotation=<rot -0.0001 0 95>}]
 cape_mid_04_r [{shapeType=Capsule, length=17.7446, radius=10.3632, shapeName=cape_mid_04_r_capsule, center=<10.1317 0 3>, rotation=<rot -0.0001 0 85>}]
 cape_mid_05_l [{shapeType=Capsule, length=17.2654, radius=9.1132, shapeName=cape_mid_05_r_capsule, center=<7.1117 0 3>, rotation=<rot -0.0001 0 95>}]
 cape_mid_05_r [{shapeType=Capsule, length=17.2654, radius=9.1132, shapeName=cape_mid_05_r_capsule, center=<6.9492 0 3>, rotation=<rot -0.0002 0 80>}]
 cape_mid_06_l [{shapeType=Capsule, length=18.9201, radius=7.6857, shapeName=cape_mid_06_l_capsule, center=<4.341 0 2>, rotation=<rot -0.0001 0 100>}]
 cape_mid_06_r [{shapeType=Capsule, length=18.6701, radius=7.1857, shapeName=cape_mid_06_r_capsule, center=<4.341 0 2>, rotation=<rot -0.0001 0 85>}]
 cape_mid_l_MSTR [{shapeType=Capsule, length=1, radius=1, shapeName=cape_mid_l_MSTR_capsule, rotation=<rot 90 0 0>}]
 cape_mid_r_MSTR [{shapeType=Capsule, length=1, radius=1, shapeName=cape_mid_r_MSTR_capsule, rotation=<rot 90 0 0>}]
 cape_r_MSTR [{shapeType=Capsule, length=1, radius=1, shapeName=cape_r_MSTR_capsule, rotation=<rot 90 0 0>}]
 clavicle_cape_l [{shapeType=Capsule, length=1, radius=1, shapeName=clavicle_cape_l_capsule, rotation=<rot 90 0 0>}]
 clavicle_cape_r [{shapeType=Capsule, length=1, radius=1, shapeName=clavicle_cape_r_capsule, rotation=<rot 90 0 0>}]
 clavicle_l [{shapeType=Capsule, length=9.9152, radius=8.0614, shapeName=clavicle_l_capsule, center=<8.31 0 0>, rotation=<rot 90 0 0>}]
 clavicle_r [{shapeType=Capsule, length=8.4152, radius=7.0614, shapeName=clavicle_r_capsule, center=<-7.9008 0 0>, rotation=<rot -90 0 0>}]
 dyn_root_slope [
  {shapeType=Sphere, radius=300, shapeName=dyn_root_slope_sphere, center=<0 -20 -295>}
  {shapeType=Sphere, radius=30.25, shapeName=dyn_root_slope_sphere, center=<0 -3.7802 0>}
 ]
 foot_l [{shapeType=Capsule, length=11.9063, radius=7.1702, shapeName=foot_l_capsule, center=<-4.831 7.2726 0>, rotation=<rot 90 180 90>}]
 foot_r [{shapeType=Capsule, length=11.9063, radius=7.1702, shapeName=foot_r_capsule, center=<4.831 -7.2726 0>, rotation=<rot 90 180 90>}]
 lowerarm_l [{
   shapeType=Capsule
   length=32.2849
   radius=10.75
   shapeName=lowerarm_l_capsule
   center=<13.1888 0.9927 -1.7353>
   rotation=<rot 85 -0.0001 -0.0002>
  }]
 lowerarm_r [{shapeType=Capsule, length=24.5349, radius=9.5, shapeName=lowerarm_r_capsule, center=<-10.0321 -2.7329 0>, rotation=<rot -90 0 0>}]
 pelvis [{shapeType=Capsule, length=5.4057, radius=13.244, shapeName=pelvis_capsule}]
 spine_01 [{shapeType=Capsule, length=7.0823, radius=11.5674, shapeName=spine_01_capsule}]
 spine_02 [{shapeType=Capsule, length=7.0855, radius=11.3382, shapeName=spine_02_capsule}]
 spine_03 [{shapeType=Capsule, length=7.4815, radius=11.1681, shapeName=spine_03_capsule}]
 spine_04 [{shapeType=Capsule, length=8.3301, radius=11.7375, shapeName=spine_04_capsule}]
 spine_05 [{shapeType=Capsule, length=11.2449, radius=12.9138, shapeName=spine_05_capsule}]
 thigh_l [{shapeType=Capsule, length=21.9725, radius=13.3748, shapeName=thigh_l_capsule, center=<-15.2556 0 0>, rotation=<rot 90 89.6476 85.2798>}]
 thigh_r [{shapeType=Capsule, length=21.9725, radius=13.3748, shapeName=thigh_r_capsule, center=<15.2556 0 0>, rotation=<rot -90 -90.3524 85.9838>}]
 upperarm_l [{
   shapeType=Capsule
   length=13.797
   radius=10.9402
   shapeName=upperarm_l_capsule
   center=<10.9919 3.2969 -1.2339>
   rotation=<rot 90 179.156 179.1562>
  }]
 upperarm_r [{
   shapeType=Capsule
   length=22.3802
   radius=8.25
   shapeName=upperarm_r_capsule
   center=<-11.3453 -0.1826 1.3767>
   rotation=<rot -85 0.001 -0.001>
  }]
}
constraints [
 {bone1Name=cape_mid_l_MSTR, bone2Name=backpack_root, swing1LimitDegrees=45, swing2LimitDegrees=45, twistLimitDegrees=0}
 {bone1Name=cape_mid_r_MSTR, bone2Name=backpack_root, swing1LimitDegrees=45, swing2LimitDegrees=45, twistLimitDegrees=0}
 {bone1Name=cape_mid_01_r, bone2Name=cape_mid_r_MSTR, swing1LimitDegrees=20, swing2LimitDegrees=10, twistLimitDegrees=0}
 {bone1Name=cape_mid_02_r, bone2Name=cape_mid_01_r, swing1LimitDegrees=20, swing2LimitDegrees=20, twistLimitDegrees=0}
 {bone1Name=cape_mid_03_r, bone2Name=cape_mid_02_r, swing1LimitDegrees=30, swing2LimitDegrees=30, twistLimitDegrees=0}
 {bone1Name=cape_mid_04_r, bone2Name=cape_mid_03_r, swing1LimitDegrees=35, swing2LimitDegrees=35, twistLimitDegrees=0}
 {bone1Name=cape_mid_05_r, bone2Name=cape_mid_04_r, swing1LimitDegrees=45, swing2LimitDegrees=45, twistLimitDegrees=0}
 {bone1Name=cape_mid_06_r, bone2Name=cape_mid_05_r, swing1LimitDegrees=45, swing2LimitDegrees=45, twistLimitDegrees=15}
 {bone1Name=clavicle_cape_l, bone2Name=spine_05, swing1LimitDegrees=45, swing2LimitDegrees=45, twistLimitDegrees=0}
 {bone1Name=cape_l_MSTR, bone2Name=clavicle_cape_l, swing1LimitDegrees=45, swing2LimitDegrees=45, twistLimitDegrees=0}
 {bone1Name=cape_01_l, bone2Name=cape_l_MSTR, swing1LimitDegrees=35, swing2LimitDegrees=35, twistLimitDegrees=15}
 {bone1Name=cape_02_l, bone2Name=cape_01_l, swing1LimitDegrees=35, swing2LimitDegrees=35, twistLimitDegrees=15}
 {bone1Name=cape_03_l, bone2Name=cape_02_l, swing1LimitDegrees=30, swing2LimitDegrees=30, twistLimitDegrees=0}
 {bone1Name=cape_04_l, bone2Name=cape_03_l, swing1LimitDegrees=35, swing2LimitDegrees=35, twistLimitDegrees=0}
 {bone1Name=cape_05_l, bone2Name=cape_04_l, swing1LimitDegrees=45, swing2LimitDegrees=45, twistLimitDegrees=0}
 {bone1Name=clavicle_cape_r, bone2Name=spine_05, swing1LimitDegrees=45, swing2LimitDegrees=45, twistLimitDegrees=0}
 {bone1Name=cape_r_MSTR, bone2Name=clavicle_cape_r, swing1LimitDegrees=45, swing2LimitDegrees=45, twistLimitDegrees=0}
 {bone1Name=cape_02_r, bone2Name=cape_01_r, swing1LimitDegrees=35, swing2LimitDegrees=35, twistLimitDegrees=15}
 {bone1Name=cape_03_r, bone2Name=cape_02_r, swing1LimitDegrees=30, swing2LimitDegrees=30, twistLimitDegrees=15}
 {bone1Name=cape_04_r, bone2Name=cape_03_r, swing1LimitDegrees=45, swing2LimitDegrees=45, twistLimitDegrees=15}
 {bone1Name=cape_05_r, bone2Name=cape_04_r, swing1LimitDegrees=45, swing2LimitDegrees=45, twistLimitDegrees=15}
 {bone1Name=cape_mid_01_l, bone2Name=cape_mid_l_MSTR, swing1LimitDegrees=20, swing2LimitDegrees=10, twistLimitDegrees=15}
 {bone1Name=cape_mid_02_l, bone2Name=cape_mid_01_l, swing1LimitDegrees=35, swing2LimitDegrees=35, twistLimitDegrees=15}
 {bone1Name=cape_mid_03_l, bone2Name=cape_mid_02_l, swing1LimitDegrees=35, swing2LimitDegrees=35, twistLimitDegrees=15}
 {bone1Name=cape_mid_04_l, bone2Name=cape_mid_03_l, swing1LimitDegrees=35, swing2LimitDegrees=35, twistLimitDegrees=0}
 {bone1Name=cape_mid_05_l, bone2Name=cape_mid_04_l, swing1LimitDegrees=45, swing2LimitDegrees=45, twistLimitDegrees=15}
 {bone1Name=cape_mid_06_l, bone2Name=cape_mid_05_l, swing1LimitDegrees=45, swing2LimitDegrees=45, twistLimitDegrees=15}
 {
  bone1Name=spine_01
  bone2Name=pelvis
  swing1LimitDegrees=45
  swing1Motion=Free
  swing2LimitDegrees=45
  swing2Motion=Free
  twistLimitDegrees=45
  twistMotion=Free
 }
 {
  bone1Name=spine_02
  bone2Name=spine_01
  swing1LimitDegrees=45
  swing1Motion=Free
  swing2LimitDegrees=45
  swing2Motion=Free
  twistLimitDegrees=45
  twistMotion=Free
 }
 {
  bone1Name=spine_03
  bone2Name=spine_02
  swing1LimitDegrees=45
  swing1Motion=Free
  swing2LimitDegrees=45
  swing2Motion=Free
  twistLimitDegrees=45
  twistMotion=Free
 }
 {
  bone1Name=spine_04
  bone2Name=spine_03
  swing1LimitDegrees=45
  swing1Motion=Free
  swing2LimitDegrees=45
  swing2Motion=Free
  twistLimitDegrees=45
  twistMotion=Free
 }
 {
  bone1Name=spine_05
  bone2Name=spine_04
  swing1LimitDegrees=45
  swing1Motion=Free
  swing2LimitDegrees=45
  swing2Motion=Free
  twistLimitDegrees=45
  twistMotion=Free
 }
 {
  bone1Name=clavicle_l
  bone2Name=spine_05
  swing1LimitDegrees=45
  swing1Motion=Free
  swing2LimitDegrees=45
  swing2Motion=Free
  twistLimitDegrees=45
  twistMotion=Free
 }
 {
  bone1Name=upperarm_l
  bone2Name=clavicle_l
  swing1LimitDegrees=45
  swing1Motion=Free
  swing2LimitDegrees=45
  swing2Motion=Free
  twistLimitDegrees=45
  twistMotion=Free
 }
 {
  bone1Name=clavicle_r
  bone2Name=spine_05
  swing1LimitDegrees=45
  swing1Motion=Free
  swing2LimitDegrees=45
  swing2Motion=Free
  twistLimitDegrees=45
  twistMotion=Free
 }
 {
  bone1Name=upperarm_r
  bone2Name=clavicle_r
  swing1LimitDegrees=45
  swing1Motion=Free
  swing2LimitDegrees=45
  swing2Motion=Free
  twistLimitDegrees=45
  twistMotion=Free
 }
 {
  bone1Name=thigh_l
  bone2Name=pelvis
  swing1LimitDegrees=45
  swing1Motion=Free
  swing2LimitDegrees=45
  swing2Motion=Free
  twistLimitDegrees=45
  twistMotion=Free
 }
 {
  bone1Name=calf_l
  bone2Name=thigh_l
  swing1LimitDegrees=45
  swing1Motion=Free
  swing2LimitDegrees=45
  swing2Motion=Free
  twistLimitDegrees=45
  twistMotion=Free
 }
 {
  bone1Name=thigh_r
  bone2Name=pelvis
  swing1LimitDegrees=45
  swing1Motion=Free
  swing2LimitDegrees=45
  swing2Motion=Free
  twistLimitDegrees=45
  twistMotion=Free
 }
 {
  bone1Name=calf_r
  bone2Name=thigh_r
  swing1LimitDegrees=45
  swing1Motion=Free
  swing2LimitDegrees=45
  swing2Motion=Free
  twistLimitDegrees=45
  twistMotion=Free
 }
 {
  bone1Name=foot_l
  bone2Name=calf_l
  swing1LimitDegrees=45
  swing1Motion=Free
  swing2LimitDegrees=45
  swing2Motion=Free
  twistLimitDegrees=45
  twistMotion=Free
 }
 {
  bone1Name=foot_r
  bone2Name=calf_r
  swing1LimitDegrees=45
  swing1Motion=Free
  swing2LimitDegrees=45
  swing2Motion=Free
  twistLimitDegrees=45
  twistMotion=Free
 }
 {
  bone1Name=lowerarm_l
  bone2Name=upperarm_l
  swing1LimitDegrees=45
  swing1Motion=Free
  swing2LimitDegrees=45
  swing2Motion=Free
  twistLimitDegrees=45
  twistMotion=Free
 }
 {
  bone1Name=lowerarm_r
  bone2Name=upperarm_r
  swing1LimitDegrees=45
  swing1Motion=Free
  swing2LimitDegrees=45
  swing2Motion=Free
  twistLimitDegrees=45
  twistMotion=Free
 }
 {bone1Name=cape_01_r, bone2Name=cape_r_MSTR, swing1LimitDegrees=35, swing2LimitDegrees=35, twistLimitDegrees=15}
 {
  bone1Name=cape_02_l
  bone2Name=cape_mid_02_l
  swing1LimitDegrees=45
  swing1Motion=Free
  swing2LimitDegrees=45
  swing2Motion=Free
  twistLimitDegrees=45
  twistMotion=Free
 }
 {
  bone1Name=cape_04_l
  bone2Name=cape_mid_04_l
  swing1LimitDegrees=45
  swing1Motion=Free
  swing2LimitDegrees=45
  swing2Motion=Free
  twistLimitDegrees=45
  twistMotion=Free
 }
 {
  bone1Name=cape_mid_02_l
  bone2Name=cape_mid_02_r
  swing1LimitDegrees=45
  swing1Motion=Free
  swing2LimitDegrees=45
  swing2Motion=Free
  twistLimitDegrees=45
  twistMotion=Free
 }
 {
  bone1Name=cape_mid_04_l
  bone2Name=cape_mid_04_r
  swing1LimitDegrees=45
  swing1Motion=Free
  swing2LimitDegrees=45
  swing2Motion=Free
  twistLimitDegrees=45
  twistMotion=Free
 }
 {
  bone1Name=cape_02_r
  bone2Name=cape_mid_02_r
  swing1LimitDegrees=45
  swing1Motion=Free
  swing2LimitDegrees=45
  swing2Motion=Free
  twistLimitDegrees=45
  twistMotion=Free
 }
 {
  bone1Name=cape_04_r
  bone2Name=cape_mid_04_r
  swing1LimitDegrees=45
  swing1Motion=Free
  swing2LimitDegrees=45
  swing2Motion=Free
  twistLimitDegrees=45
  twistMotion=Free
 }
 {
  bone1Name=backpack_root
  bone2Name=spine_05
  swing1LimitDegrees=45
  swing1Motion=Free
  swing2LimitDegrees=45
  swing2Motion=Free
  twistLimitDegrees=45
  twistMotion=Free
 }
]
