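# Physics asset for the M_MED_Mystic collar, apparently the crouching-with-backbling variant, simulating four cape chains and a necklace against body capsules.
# Cape chains hang from spine_05 via master bones, the outer pair through clavicle_cape bones, with narrow swing limits along each chain.
# Body joints and two links between neighboring cape chains leave rotation free.
# A huge root capsule forms a ground-like collider; spine_05 carries an extra large offset capsule.
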PhysicsAsset /BRCosmetics/Characters/Player/Male/Medium/Bodies/M_MED_Mystic/Meshes/M_MED_Mystic_Collar_Crouching_BackblingPhysics
bodies {
 calf_l [{shapeType=Capsule, length=33.2834, radius=5.5628, shapeName=calf_l_capsule, center=<-22.462 -1.353 0>, rotation=<rot 90 83.2785 92.2051>}]
 calf_r [{shapeType=Capsule, length=30.0334, radius=8.8128, shapeName=calf_r_capsule, center=<21.7328 -1.7569 0>, rotation=<rot -90 180 166.8032>}]
 cape_01_l [{shapeType=Capsule, length=28.5, radius=5.5, shapeName=cape_01_l_capsule, center=<10.6583 0 -0.8595>, rotation=<rot 90 180 -90>}]
 cape_01_r [{shapeType=Capsule, length=27.2362, radius=6.5, shapeName=cape_01_r_capsule, center=<10.5952 0 3.1493>, rotation=<rot 0 0 90>}]
 cape_02_l [
  {shapeType=Capsule, length=24.9097, radius=3.6181, shapeName=cape_02_l_capsule, center=<13.709 0 0>, rotation=<rot 90 0 90>}
  {shapeType=Capsule, length=25.25, radius=4.75, shapeName=cape_02_l_capsule, center=<12.8625 0 0>, rotation=<rot 90 0 0>}
 ]
 cape_02_r [
  {shapeType=Capsule, length=21.1597, radius=3.6181, shapeName=cape_02_r_capsule, center=<8.4829 0 0>, rotation=<rot 0 0 90>}
  {shapeType=Capsule, length=27, radius=3, shapeName=cape_02_r_capsule, center=<12.8625 0 0>, rotation=<rot 90 0 0>}
 ]
 cape_03_l [{shapeType=Capsule, length=15.441, radius=6.255, shapeName=cape_03_l_capsule, center=<11.9516 0 0>, rotation=<rot 90 0 0>}]
 cape_03_r [{shapeType=Capsule, length=15.441, radius=6.255, shapeName=cape_03_r_capsule, center=<11.9516 0 0>, rotation=<rot 90 0 0>}]
 cape_04_l [{shapeType=Capsule, length=17.0439, radius=3.8632, shapeName=cape_04_l_capsule, center=<11.9516 0 0>, rotation=<rot 90 0 0>}]
 cape_04_r [{shapeType=Capsule, length=6.2939, radius=9.1132, shapeName=cape_04_r_capsule, center=<11.9516 0 0>, rotation=<rot 90 0 0>}]
 cape_05_l [{shapeType=Capsule, length=10.7889, radius=3.9839, shapeName=cape_05_l_capsule, center=<6.8829 0 0>, rotation=<rot 90 0 0>}]
 cape_05_r [{shapeType=Capsule, length=3.2889, radius=7.4839, shapeName=cape_05_r_capsule, center=<6.8829 0 0>, rotation=<rot 90 0 0>}]
 cape_l_MSTR [{shapeType=Capsule, length=1, radius=1, shapeName=cape_l_MSTR_capsule, rotation=<rot 90 0 0>}]
 cape_mid_01_l [
  {shapeType=Capsule, length=12.0944, radius=4.5403, shapeName=cape_mid_01_r_capsule, center=<7.4618 0 5.8961>, rotation=<rot 90 180 0>}
  {shapeType=Capsule, length=38, radius=4.2, shapeName=cape_mid_01_r_capsule, center=<15.824 -0.0749 2.512>, rotation=<rot 0 0 88.2923>}
 ]
 cape_mid_01_r [
  {shapeType=Capsule, length=13.5944, radius=4.5403, shapeName=cape_mid_01_r_capsule, center=<7.4618 0 -2.0929>, rotation=<rot 90 180 0>}
  {shapeType=Capsule, length=38, radius=4.2, shapeName=cape_mid_01_r_capsule, center=<15.824 0.0607 -2.0352>, rotation=<rot 0 0 88.2923>}
 ]
 cape_mid_02_l [{shapeType=Capsule, length=22.115, radius=3.4992, shapeName=cape_mid_02_r_capsule, center=<13.0851 0 0>, rotation=<rot 90 180 -90>}]
 cape_mid_02_r [{shapeType=Capsule, length=22.115, radius=5.4992, shapeName=cape_mid_02_r_capsule, center=<13.0851 0 0>, rotation=<rot 90 180 -90>}]
 cape_mid_03_l [{shapeType=Capsule, length=19.691, radius=8.755, shapeName=cape_mid_03_r_capsule, center=<11.9516 0 0>, rotation=<rot 90 0 0>}]
 cape_mid_03_r [{shapeType=Capsule, length=24.191, radius=8.255, shapeName=cape_mid_03_r_capsule, center=<11.9516 0 0>, rotation=<rot 90 0 0>}]
 cape_mid_04_l [{shapeType=Capsule, length=23.2446, radius=4.8632, shapeName=cape_mid_04_r_capsule, center=<11.9516 0 0>, rotation=<rot 90 0 90>}]
 cape_mid_04_r [{shapeType=Capsule, length=23.2446, radius=4.8632, shapeName=cape_mid_04_r_capsule, center=<11.9516 0 0>, rotation=<rot 90 0 90>}]
 cape_mid_05_l [{shapeType=Capsule, length=17.2654, radius=6.8632, shapeName=cape_mid_05_r_capsule, center=<11.9516 0 0>, rotation=<rot 90 0 90>}]
 cape_mid_05_r [{shapeType=Capsule, length=17.2654, radius=9.1132, shapeName=cape_mid_05_r_capsule, center=<11.9516 0 0>, rotation=<rot 90 0 90>}]
 cape_mid_06_l [{shapeType=Capsule, length=2.6701, radius=7.6857, shapeName=cape_mid_06_l_capsule, center=<4.341 0 0>, rotation=<rot 90 0 0>}]
 cape_mid_06_r [{shapeType=Capsule, length=2.6701, radius=6.6857, shapeName=cape_mid_06_r_capsule, center=<4.341 0 0>, rotation=<rot 90 0 0>}]
 cape_mid_l_MSTR [{shapeType=Capsule, length=1, radius=1, shapeName=cape_mid_l_MSTR_capsule, rotation=<rot 90 0 0>}]
 cape_mid_r_MSTR [{shapeType=Capsule, length=1, radius=1, shapeName=cape_mid_r_MSTR_capsule, rotation=<rot 90 0 0>}]
 cape_r_MSTR [{shapeType=Capsule, length=1, radius=1, shapeName=cape_r_MSTR_capsule, rotation=<rot 90 0 0>}]
 clavicle_cape_l [{shapeType=Capsule, length=1, radius=1, shapeName=clavicle_cape_l_capsule, rotation=<rot 90 0 0>}]
 clavicle_cape_r [{shapeType=Capsule, length=1, radius=1, shapeName=clavicle_cape_r_capsule, rotation=<rot 90 0 0>}]
 clavicle_l [{shapeType=Capsule, length=8.6652, radius=6.8114, shapeName=clavicle_l_capsule, center=<7.9008 0 0>, rotation=<rot 90 0 0>}]
 clavicle_r [{shapeType=Capsule, length=8.6652, radius=6.8114, shapeName=clavicle_r_capsule, center=<-7.9008 0 0>, rotation=<rot -90 0 0>}]
 dyn_necklace_01 [{
   shapeType=Capsule
   length=2.4187
   radius=1.1163
   shapeName=dyn_necklace_01_capsule
   center=<1.0317 -0.2687 0.1829>
   rotation=<rot 0 175.4092 -89.0153>
  }]
 foot_l [{shapeType=Capsule, length=14.1563, radius=4.9202, shapeName=foot_l_capsule, center=<-4.831 7.2726 0>, rotation=<rot 90 180 90>}]
 foot_r [{shapeType=Capsule, length=11.9063, radius=7.1702, shapeName=foot_r_capsule, center=<4.831 -7.2726 0>, rotation=<rot 90 180 90>}]
 lowerarm_l [{shapeType=Capsule, length=19.2849, radius=5, shapeName=lowerarm_l_capsule, center=<10.0321 0 0>, rotation=<rot 90 0 0>}]
 lowerarm_r [{shapeType=Capsule, length=20.7849, radius=5.25, shapeName=lowerarm_r_capsule, center=<-10.0321 0 0>, rotation=<rot -90 180 180>}]
 pelvis [{shapeType=Capsule, length=5.4057, radius=13.244, shapeName=pelvis_capsule}]
 root [{shapeType=Capsule, length=0, radius=300, shapeName=root_capsule, center=<0 -19.1024 -296.7835>, rotation=<rot 0 -90 0>}]
 spine_01 [{shapeType=Capsule, length=7.0823, radius=11.5674, shapeName=spine_01_capsule}]
 spine_02 [{shapeType=Capsule, length=7.5855, radius=10.8382, shapeName=spine_02_capsule}]
 spine_03 [{shapeType=Capsule, length=8.2315, radius=10.4181, shapeName=spine_03_capsule}]
 spine_04 [{shapeType=Capsule, length=9.3301, radius=10.7375, shapeName=spine_04_capsule}]
 spine_05 [
  {shapeType=Capsule, length=11.7449, radius=12.4138, shapeName=spine_05_capsule}
  {
   shapeType=Capsule
   length=37
   radius=23.75
   shapeName=spine_05_capsule
   center=<-10.2523 16.0945 51.7531>
   rotation=<rot 5.5651 179.9999 -148.5051>
  }
 ]
 thigh_l [{shapeType=Capsule, length=21.9725, radius=13.3748, shapeName=thigh_l_capsule, center=<-15.2556 0 0>, rotation=<rot 90 89.6476 85.2798>}]
 thigh_r [{shapeType=Capsule, length=21.9725, radius=13.3748, shapeName=thigh_r_capsule, center=<15.2556 0 0>, rotation=<rot -90 -90.3524 85.9838>}]
 upperarm_l [{shapeType=Capsule, length=24.047, radius=5.9402, shapeName=upperarm_l_capsule, center=<9.5135 0 0>, rotation=<rot 90 0 0.0002>}]
 upperarm_r [{shapeType=Capsule, length=25.8802, radius=5.5, shapeName=upperarm_r_capsule, center=<-10.5423 0 0>, rotation=<rot -90 180 180>}]
}
constraints [
 {bone1Name=dyn_necklace_01, bone2Name=spine_05, swing1LimitDegrees=20, swing2LimitDegrees=20, twistLimitDegrees=0}
 {bone1Name=cape_mid_l_MSTR, bone2Name=spine_05, swing1LimitDegrees=45, swing2LimitDegrees=45, twistLimitDegrees=0}
 {bone1Name=cape_mid_r_MSTR, bone2Name=spine_05, swing1LimitDegrees=45, swing2LimitDegrees=45, twistLimitDegrees=0}
 {bone1Name=cape_mid_01_r, bone2Name=cape_mid_r_MSTR, swing1LimitDegrees=20, swing2LimitDegrees=10, twistLimitDegrees=0}
 {bone1Name=cape_mid_02_r, bone2Name=cape_mid_01_r, swing1LimitDegrees=10, swing2LimitDegrees=10, twistLimitDegrees=0}
 {bone1Name=cape_mid_03_r, bone2Name=cape_mid_02_r, swing1LimitDegrees=20, swing2LimitDegrees=20, twistLimitDegrees=0}
 {bone1Name=cape_mid_04_r, bone2Name=cape_mid_03_r, swing1LimitDegrees=10, swing2LimitDegrees=10, twistLimitDegrees=0}
 {bone1Name=cape_mid_05_r, bone2Name=cape_mid_04_r, swing1LimitDegrees=20, swing2LimitDegrees=20, twistLimitDegrees=0}
 {bone1Name=cape_mid_06_r, bone2Name=cape_mid_05_r, swing1LimitDegrees=20, swing2LimitDegrees=20, twistLimitDegrees=0}
 {bone1Name=clavicle_cape_l, bone2Name=spine_05, swing1LimitDegrees=45, swing2LimitDegrees=45, twistLimitDegrees=0}
 {bone1Name=cape_l_MSTR, bone2Name=clavicle_cape_l, swing1LimitDegrees=45, swing2LimitDegrees=45, twistLimitDegrees=0}
 {bone1Name=cape_01_l, bone2Name=cape_l_MSTR, swing1LimitDegrees=10, swing2LimitDegrees=10, twistLimitDegrees=0}
 {bone1Name=cape_02_l, bone2Name=cape_01_l, swing1LimitDegrees=10, swing2LimitDegrees=10, twistLimitDegrees=0}
 {bone1Name=cape_03_l, bone2Name=cape_02_l, swing1LimitDegrees=10, swing2LimitDegrees=10, twistLimitDegrees=0}
 {bone1Name=cape_04_l, bone2Name=cape_03_l, swing1LimitDegrees=10, swing2LimitDegrees=10, twistLimitDegrees=0}
 {bone1Name=cape_05_l, bone2Name=cape_04_l, swing1LimitDegrees=10, swing2LimitDegrees=10, twistLimitDegrees=0}
 {bone1Name=clavicle_cape_r, bone2Name=spine_05, swing1LimitDegrees=45, swing2LimitDegrees=45, twistLimitDegrees=0}
 {bone1Name=cape_r_MSTR, bone2Name=clavicle_cape_r, swing1LimitDegrees=45, swing2LimitDegrees=45, twistLimitDegrees=0}
 {bone1Name=cape_02_r, bone2Name=cape_01_r, swing1LimitDegrees=20, swing2LimitDegrees=20, twistLimitDegrees=0}
 {bone1Name=cape_03_r, bone2Name=cape_02_r, swing1LimitDegrees=20, swing2LimitDegrees=20, twistLimitDegrees=0}
 {bone1Name=cape_04_r, bone2Name=cape_03_r, swing1LimitDegrees=20, swing2LimitDegrees=20, twistLimitDegrees=0}
 {bone1Name=cape_05_r, bone2Name=cape_04_r, swing1LimitDegrees=20, swing2LimitDegrees=20, twistLimitDegrees=0}
 {bone1Name=cape_mid_01_l, bone2Name=cape_mid_l_MSTR, swing1LimitDegrees=20, swing2LimitDegrees=10, twistLimitDegrees=0}
 {bone1Name=cape_mid_02_l, bone2Name=cape_mid_01_l, swing1LimitDegrees=20, swing2LimitDegrees=20, twistLimitDegrees=0}
 {bone1Name=cape_mid_03_l, bone2Name=cape_mid_02_l, swing1LimitDegrees=12, swing2LimitDegrees=12, twistLimitDegrees=0}
 {bone1Name=cape_mid_04_l, bone2Name=cape_mid_03_l, swing1LimitDegrees=12, swing2LimitDegrees=12, twistLimitDegrees=0}
 {bone1Name=cape_mid_05_l, bone2Name=cape_mid_04_l, swing1LimitDegrees=20, swing2LimitDegrees=20, twistLimitDegrees=0}
 {bone1Name=cape_mid_06_l, bone2Name=cape_mid_05_l, swing1LimitDegrees=20, swing2LimitDegrees=20, twistLimitDegrees=0}
 {
  bone1Name=spine_01
  bone2Name=pelvis
  swing1LimitDegrees=45
  swing1Motion=Free
  swing2LimitDegrees=45
  swing2Motion=Free
  twistLimitDegrees=45
  twistMotion=Free
 }
 {
  bone1Name=spine_02
  bone2Name=spine_01
  swing1LimitDegrees=45
  swing1Motion=Free
  swing2LimitDegrees=45
  swing2Motion=Free
  twistLimitDegrees=45
  twistMotion=Free
 }
 {
  bone1Name=spine_03
  bone2Name=spine_02
  swing1LimitDegrees=45
  swing1Motion=Free
  swing2LimitDegrees=45
  swing2Motion=Free
  twistLimitDegrees=45
  twistMotion=Free
 }
 {
  bone1Name=spine_04
  bone2Name=spine_03
  swing1LimitDegrees=45
  swing1Motion=Free
  swing2LimitDegrees=45
  swing2Motion=Free
  twistLimitDegrees=45
  twistMotion=Free
 }
 {
  bone1Name=spine_05
  bone2Name=spine_04
  swing1LimitDegrees=45
  swing1Motion=Free
  swing2LimitDegrees=45
  swing2Motion=Free
  twistLimitDegrees=45
  twistMotion=Free
 }
 {
  bone1Name=clavicle_l
  bone2Name=spine_05
  swing1LimitDegrees=45
  swing1Motion=Free
  swing2LimitDegrees=45
  swing2Motion=Free
  twistLimitDegrees=45
  twistMotion=Free
 }
 {
  bone1Name=upperarm_l
  bone2Name=clavicle_l
  swing1LimitDegrees=45
  swing1Motion=Free
  swing2LimitDegrees=45
  swing2Motion=Free
  twistLimitDegrees=45
  twistMotion=Free
 }
 {
  bone1Name=clavicle_r
  bone2Name=spine_05
  swing1LimitDegrees=45
  swing1Motion=Free
  swing2LimitDegrees=45
  swing2Motion=Free
  twistLimitDegrees=45
  twistMotion=Free
 }
 {
  bone1Name=upperarm_r
  bone2Name=clavicle_r
  swing1LimitDegrees=45
  swing1Motion=Free
  swing2LimitDegrees=45
  swing2Motion=Free
  twistLimitDegrees=45
  twistMotion=Free
 }
 {
  bone1Name=thigh_l
  bone2Name=pelvis
  swing1LimitDegrees=45
  swing1Motion=Free
  swing2LimitDegrees=45
  swing2Motion=Free
  twistLimitDegrees=45
  twistMotion=Free
 }
 {
  bone1Name=calf_l
  bone2Name=thigh_l
  swing1LimitDegrees=45
  swing1Motion=Free
  swing2LimitDegrees=45
  swing2Motion=Free
  twistLimitDegrees=45
  twistMotion=Free
 }
 {
  bone1Name=thigh_r
  bone2Name=pelvis
  swing1LimitDegrees=45
  swing1Motion=Free
  swing2LimitDegrees=45
  swing2Motion=Free
  twistLimitDegrees=45
  twistMotion=Free
 }
 {
  bone1Name=calf_r
  bone2Name=thigh_r
  swing1LimitDegrees=45
  swing1Motion=Free
  swing2LimitDegrees=45
  swing2Motion=Free
  twistLimitDegrees=45
  twistMotion=Free
 }
 {
  bone1Name=pelvis
  bone2Name=root
  swing1LimitDegrees=45
  swing1Motion=Free
  swing2LimitDegrees=45
  swing2Motion=Free
  twistLimitDegrees=45
  twistMotion=Free
 }
 {
  bone1Name=foot_l
  bone2Name=calf_l
  swing1LimitDegrees=45
  swing1Motion=Free
  swing2LimitDegrees=45
  swing2Motion=Free
  twistLimitDegrees=45
  twistMotion=Free
 }
 {
  bone1Name=foot_r
  bone2Name=calf_r
  swing1LimitDegrees=45
  swing1Motion=Free
  swing2LimitDegrees=45
  swing2Motion=Free
  twistLimitDegrees=45
  twistMotion=Free
 }
 {
  bone1Name=lowerarm_l
  bone2Name=upperarm_l
  swing1LimitDegrees=45
  swing1Motion=Free
  swing2LimitDegrees=45
  swing2Motion=Free
  twistLimitDegrees=45
  twistMotion=Free
 }
 {
  bone1Name=lowerarm_r
  bone2Name=upperarm_r
  swing1LimitDegrees=45
  swing1Motion=Free
  swing2LimitDegrees=45
  swing2Motion=Free
  twistLimitDegrees=45
  twistMotion=Free
 }
 {bone1Name=cape_01_r, bone2Name=cape_r_MSTR, swing1LimitDegrees=20, swing2LimitDegrees=20, twistLimitDegrees=0}
 {
  bone1Name=cape_mid_02_r
  bone2Name=cape_02_r
  swing1LimitDegrees=45
  swing1Motion=Free
  swing2LimitDegrees=45
  swing2Motion=Free
  twistLimitDegrees=45
  twistMotion=Free
 }
 {
  bone1Name=cape_04_l
  bone2Name=cape_mid_04_l
  swing1LimitDegrees=45
  swing1Motion=Free
  swing2LimitDegrees=45
  swing2Motion=Free
  twistLimitDegrees=45
  twistMotion=Free
 }
]
