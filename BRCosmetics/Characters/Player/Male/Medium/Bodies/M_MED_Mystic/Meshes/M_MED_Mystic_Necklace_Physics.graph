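# Physics asset for the necklace mesh of the M_MED_Mystic male player body, set up for a dangling pendant against the upper torso.
# Two large spine_05 capsules approximate the torso, dyn_necklace_01 carries the necklace capsule, and neck_01 and dyn_necklace_MSTR hold tiny capsules.
# The necklace bone is constrained to its master with narrow, uneven swing limits and zero twist limit.
# The master and neck join spine_05 through free angular constraints.
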PhysicsAsset /BRCosmetics/Characters/Player/Male/Medium/Bodies/M_MED_Mystic/Meshes/M_MED_Mystic_Necklace_Physics
bodies {
 dyn_necklace_01 [{
   shapeType=Capsule
   length=6.0411
   radius=2.3611
   shapeName=dyn_necklace_01_capsule
   center=<7.8628 0 1.3076>
   rotation=<rot -69.4773 180 -180>
  }]
 dyn_necklace_MSTR [{shapeType=Capsule, length=0.505, radius=0.505, shapeName=dyn_necklace_MSTR_capsule, rotation=<rot 0 -90 0>}]
 neck_01 [{shapeType=Capsule, length=0.505, radius=0.505, shapeName=neck_01_capsule, rotation=<rot 0 -90 0>}]
 spine_05 [
  {
   shapeType=Capsule
   length=10.1885
   radius=20.7305
   shapeName=spine_05_capsule
   center=<-7.9137 -9.5807 -0.1166>
   rotation=<rot -1.5405 -90 0.8273>
  }
  {
   shapeType=Capsule
   length=20.1978
   radius=9.8022
   shapeName=spine_05_capsule
   center=<-0.766 2.8096 -12.7847>
   rotation=<rot 0.0329 -15.0792 101.1592>
  }
 ]
}
constraints [
 {
  bone1Name=neck_01
  bone2Name=spine_05
  swing1LimitDegrees=45
  swing1Motion=Free
  swing2LimitDegrees=45
  swing2Motion=Free
  twistLimitDegrees=45
  twistMotion=Free
 }
 {bone1Name=dyn_necklace_01, bone2Name=dyn_necklace_MSTR, swing1LimitDegrees=5, swing2LimitDegrees=12, twistLimitDegrees=0}
 {
  bone1Name=dyn_necklace_MSTR
  bone2Name=spine_05
  swing1LimitDegrees=45
  swing1Motion=Free
  swing2LimitDegrees=45
  swing2Motion=Free
  twistLimitDegrees=45
  twistMotion=Free
 }
]
