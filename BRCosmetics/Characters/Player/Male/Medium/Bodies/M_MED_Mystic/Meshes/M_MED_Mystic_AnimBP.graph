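# Animation Blueprint for the M_MED_Mystic player cosmetic body, layering procedural cape, belt, lapel, collar and ring motion over the pose copied from the parent mesh.
# Control Rigs deform the copied pose, then Anim Dynamics, constraints and crouch offsets adjust it.
# Crouch- and backbling-dependent RigidBody variants drive secondary motion, with trail controllers as the non-rigid-body alternative.
# Cached poses feed blends selected by rigid-body and DBNO state, ahead of final belt and lapel tweaks.
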AnimBlueprint /BRCosmetics/Characters/Player/Male/Medium/Bodies/M_MED_Mystic/Meshes/M_MED_Mystic_AnimBP
target_skeleton=&/M_MED_Mystic_Skeleton
graphs [
 AnimGraph {nodes [
   n1 = Get isNOTCrouchingFloat() -> (isNOTCrouchingFloat:double)
   "Constraint - dyn_belt_01 (RR)"!(ComponentPose, Alpha=@n1, ConstraintWeights_0=0.5, ConstraintWeights_1:float=0.5) -> (Pose) {
    alphaBoolBlend {blendOption=Linear}
    boneToModify {boneName=dyn_belt_01}
    constraintSetup [{transformType=Rotation, targetBone {boneName=thigh_l}}, {transformType=Rotation, targetBone {boneName=thigh_twist_01_l}}]
    constraintWeights [1,1]
   }
   n2 = Get bIsMoving2D()
   "Transform (Modify) Bone - Bone: cape_mid_r_MSTR"!(ComponentPose, bAlphaBoolEnabled:bool=@n2, Translation="0,0,3", Rotation="10,0,0", Scale="1,1,1") -> (Pose) {
    alphaInputType=Bool
    rotationMode=BMM_Additive
    rotationSpace=BCS_BoneSpace
    translationMode=BMM_Additive
    translationSpace=BCS_BoneSpace
    alphaBoolBlend {blendInTime=0.1, blendOption=Linear, blendOutTime=0.1}
    boneToModify {boneName=cape_mid_r_MSTR}
    rotation {pitch=10}
    translation {z=3}
   }
   n3 = "Transform (Modify) Bone - Bone: clavicle_cape_l"!(
    ComponentPose=Local To Component(LocalPose=Use cached pose 'RbanCache'())
    bAlphaBoolEnabled:bool=Get bIsPlayingMeleeAnim()
    Translation="0,0,0"
    Rotation="0,-20,0"
    Scale="1,1,1"
   ) -> (Pose) {
    alphaInputType=Bool
    rotationMode=BMM_Additive
    rotationSpace=BCS_BoneSpace
    alphaBoolBlend {blendInTime=0.25, blendOption=Linear, blendOutTime=0.25}
    boneToModify {boneName=clavicle_cape_l}
    rotation {yaw=-20}
   }
   n4 = Blend Poses by bool(BlendPose_0=Component To Local(ComponentPose=@n3), BlendPose_1=Use cached pose 'trailCache'(), bActiveValue=Get bRigidBodyEnabled()) {blendType=Linear}
   n5 = Use cached pose 'dbnoCache'()
   n6 = Get bIsDBNO()
   n7 = "Transform (Modify) Bone - Bone: dyn_lapel_front_l_01"(
    ComponentPose=Local To Component(LocalPose=Blend Poses by bool(BlendPose_0=@n5, BlendPose_1=@n4, bActiveValue=@n6) {blendType=Linear})
    Alpha=1
    Translation="0,3,0"
    Rotation="-1,0,-4"
    Scale="1,1,0.8"
   ) {
    rotationMode=BMM_Additive
    scaleMode=BMM_Additive
    scaleSpace=BCS_BoneSpace
    translationMode=BMM_Additive
    alphaBoolBlend {blendOption=Linear}
    boneToModify {boneName=dyn_lapel_front_l_01}
    rotation {pitch=-1, roll=-4}
    scale {z=0.8}
    translation {y=3}
   }
   n8 = "Transform (Modify) Bone - Bone: dyn_lapel_front_l_02"(ComponentPose=@n7, Alpha=1, Translation="0,0,0", Rotation="-5,0,5", Scale="1,1,0.8") {
    rotationMode=BMM_Additive
    scaleMode=BMM_Additive
    scaleSpace=BCS_BoneSpace
    translationMode=BMM_Additive
    alphaBoolBlend {blendOption=Linear}
    boneToModify {boneName=dyn_lapel_front_l_02}
    rotation {pitch=-5, roll=5}
    scale {z=0.8}
   }
   n9 = "Transform (Modify) Bone - Bone: dyn_belt_MSTR"!(ComponentPose=@n8, bAlphaBoolEnabled:bool=True, Translation="0,0,0", Rotation="5,0,-5", Scale="1,1,1") -> (Pose) {
    alphaInputType=Bool
    rotationMode=BMM_Additive
    scaleMode=BMM_Additive
    scaleSpace=BCS_BoneSpace
    translationMode=BMM_Additive
    alphaBoolBlend {blendInTime=0.1, blendOption=Linear, blendOutTime=0.1}
    boneToModify {boneName=dyn_belt_MSTR}
    rotation {pitch=5, roll=-5}
   }
   n10 = "Transform (Modify) Bone - Bone: dyn_belt_03"!(ComponentPose=@n9, bAlphaBoolEnabled:bool=True, Translation="0,0,0", Rotation="-0.000,0.000,15.000", Scale="1,1,1") -> (Pose) {
    alphaInputType=Bool
    rotationMode=BMM_Additive
    scaleMode=BMM_Additive
    scaleSpace=BCS_BoneSpace
    translationMode=BMM_Additive
    alphaBoolBlend {blendInTime=0.1, blendOption=Linear, blendOutTime=0.1}
    boneToModify {boneName=dyn_belt_03}
    rotation {roll=15}
   }
   Output Pose(Result=Component To Local(ComponentPose=@n10)) {layerGroup=None}
   n11 = Use cached pose 'FinalBodyCache'()
   Save cached pose 'Start_Cache'(Pose=Component To Local(ComponentPose=Local To Component(LocalPose=@n11))) {cacheName=Start_Cache}
   n12 = Get consumable() -> (consumable:bool)
   n13 = Look At - Bone: cape_r_MSTR!(ComponentPose=Local To Component(LocalPose=Use cached pose 'Start_Cache'()), bAlphaBoolEnabled:bool=@n12) -> (Pose) {
    alphaInputType=Bool
    alphaBoolBlend {blendInTime=0.1, blendOption=Linear, blendOutTime=0.1}
    boneToModify {boneName=cape_r_MSTR}
    lookAtLocation=<68.5553 36.2556 -15.9758>
    lookAtTarget {boneReference {boneName=lowerarm_r}}
    lookAt_Axis {axis=<1 0>}
   }
   n14 = Get bIsTargetingConsumableThrow() -> (bIsTargetingConsumableThrow:bool)
   n15 = "Transform (Modify) Bone - Bone: clavicle_cape_r"!(ComponentPose=@n13, bAlphaBoolEnabled:bool=@n14, Translation="0,0,0", Rotation="-0.000,25,0.000", Scale="1,1,1") -> (Pose) {
    alphaInputType=Bool
    rotationMode=BMM_Additive
    rotationSpace=BCS_BoneSpace
    alphaBoolBlend {blendInTime=0.1, blendOption=Linear, blendOutTime=0.1}
    boneToModify {boneName=clavicle_cape_r}
    rotation {yaw=25}
   }
   n16 = "Transform (Modify) Bone - Bone: cape_r_MSTR"!(ComponentPose=@n15, bAlphaBoolEnabled:bool=@n14, Translation="0,0,0", Rotation="-30,0,0", Scale="1,1,1") -> (Pose) {
    alphaInputType=Bool
    rotationMode=BMM_Additive
    rotationSpace=BCS_BoneSpace
    alphaBoolBlend {blendInTime=0.1, blendOption=Linear, blendOutTime=0.1}
    boneToModify {boneName=cape_r_MSTR}
    rotation {pitch=-30}
   }
   n17 = Get bIsJumping()
   n18 = Get bIsFalling()
   n19 = "Transform (Modify) Bone - Bone: clavicle_cape_r"(
    ComponentPose=@n16
    bAlphaBoolEnabled=AND Boolean(A=Get consumable(), B=OR Boolean(A=@n17, B=@n18))
    Translation="0,0,0"
    Rotation="-0.000,25,0.000"
    Scale="1,1,1"
   ) {
    alphaInputType=Bool
    rotationMode=BMM_Additive
    rotationSpace=BCS_BoneSpace
    alphaBoolBlend {blendInTime=0.3, blendOption=Linear, blendOutTime=0.2}
    boneToModify {boneName=clavicle_cape_r}
    rotation {yaw=25}
   }
   n20 = "Constraint - cape_02_r (R)"!(ComponentPose=@n19, bAlphaBoolEnabled:bool=@n12, ConstraintWeights_0=1) -> (Pose) {
    alphaInputType=Bool
    alphaBoolBlend {blendInTime=0.1, blendOption=Linear, blendOutTime=0.1}
    boneToModify {boneName=cape_02_r}
    constraintSetup [{transformType=Rotation, targetBone {boneName=cape_mid_02_r}}]
    constraintWeights [1]
   }
   n21 = Get TrailControllerRot() -> (TrailControllerRot:Rotator)
   n22 = "Transform (Modify) Bone - Bone: cape_mid_l_MSTR"(ComponentPose=@n20, Alpha=1, Translation="0,0,0", Rotation=@n21, Scale="1,1,1") {rotationMode=BMM_Additive, rotationSpace=BCS_BoneSpace, alphaBoolBlend {blendOption=Linear}, boneToModify {boneName=cape_mid_l_MSTR}}
   n23 = "Transform (Modify) Bone - Bone: cape_mid_r_MSTR"!(ComponentPose=@n22, Alpha:float=1, Translation="0,0,0", Rotation=@n21, Scale="1,1,1") -> (Pose) {rotationMode=BMM_Additive, rotationSpace=BCS_BoneSpace, alphaBoolBlend {blendOption=Linear}, boneToModify {boneName=cape_mid_r_MSTR}}
   n24 = "Transform (Modify) Bone - Bone: cape_02_l"(ComponentPose=@n23, Alpha=1, Translation="0,0,0", Rotation=@n21, Scale="1,1,1") {rotationMode=BMM_Additive, rotationSpace=BCS_BoneSpace, alphaBoolBlend {blendOption=Linear}, boneToModify {boneName=cape_02_l}}
   n25 = "Transform (Modify) Bone - Bone: cape_02_r"(ComponentPose=@n24, Alpha=1, Translation="0,0,0", Rotation=@n21, Scale="1,1,1") {rotationMode=BMM_Additive, rotationSpace=BCS_BoneSpace, alphaBoolBlend {blendOption=Linear}, boneToModify {boneName=cape_02_r}}
   n26 = Get BwdMovement() -> (BwdMovement:double)
   n27 = Trail controller - Bone: cape_mid_06_l(ComponentPose=@n25, Alpha=@n26) {
    bUsePlanarLimit=true
    chainLength=5
    debugLifeTime=0.1
    maxDeltaTime=0.333
    alphaScaleBiasClamp {bInterpResult=true}
    rotationLimits [{limitMax=<180 180 180>}, {limitMax=<180 180 180>}, {limitMax=<180 180 180>}, {limitMax=<180 180 180>}, {limitMax=<180 180 180>}]
    rotationOffsets [<0 0 0>, <0 0 0>, <0 0 0>, <0 0 0>, <0 0 0>]
    trailBone {boneName=cape_mid_06_l}
    trailRelaxationSpeed {editorCurveData {keys [<at -0.0134 19.3841 RCIM_Linear>,<at 0.5079 12.8986 RCIM_Linear>,<at 0.9955 9.6044 RCIM_Linear>]}}
   }
   n28 = Trail controller - Bone: cape_mid_06_r(ComponentPose=@n27, Alpha=@n26) {
    bUsePlanarLimit=true
    chainLength=5
    debugLifeTime=0.1
    maxDeltaTime=0.0333
    alphaScaleBiasClamp {bInterpResult=true}
    rotationLimits [
     {limitMax=<180 180 180>}
     {limitMax=<180 180 180>}
     {limitMax=<180 90 180>, limitMin=<-180 -90 -180>}
     {limitMax=<180 180 180>}
     {limitMax=<180 180 180>}
    ]
    rotationOffsets [<0 0 0>, <0 0 0>, <0 0 0>, <0 0 0>, <0 0 0>]
    trailBone {boneName=cape_mid_06_r}
    trailRelaxationSpeed {editorCurveData {keys [<at 0 15 RCIM_Linear>,<at 0.5289 13.1838 RCIM_Linear>,<at 1 6 RCIM_Linear>]}}
   }
   n29 = Trail controller - Bone: cape_05_l(ComponentPose=@n28, Alpha=@n26) {
    bUsePlanarLimit=true
    chainLength=5
    debugLifeTime=0.1
    maxDeltaTime=0.333
    alphaScaleBiasClamp {bInterpResult=true}
    rotationLimits [{limitMax=<180 180 180>}, {limitMax=<180 180 180>}, {limitMax=<180 180 180>}, {limitMax=<180 180 180>}, {limitMax=<180 180 180>}]
    rotationOffsets [<0 0 0>, <0 0 0>, <0 0 0>, <0 0 0>, <0 0 0>]
    trailBone {boneName=cape_05_l}
    trailRelaxationSpeed {editorCurveData {keys [<at -0.0134 19.3841 RCIM_Linear>,<at 0.5079 12.8986 RCIM_Linear>,<at 0.9955 9.6044 RCIM_Linear>]}}
   }
   n30 = Trail controller - Bone: cape_05_r(ComponentPose=@n29, Alpha=@n26) {
    bUsePlanarLimit=true
    chainLength=5
    debugLifeTime=0.1
    maxDeltaTime=0.0333
    alphaScaleBiasClamp {bInterpResult=true}
    rotationLimits [
     {limitMax=<180 180 180>}
     {limitMax=<180 180 180>}
     {limitMax=<180 90 180>, limitMin=<-180 -90 -180>}
     {limitMax=<180 180 180>}
     {limitMax=<180 180 180>}
    ]
    rotationOffsets [<0 0 0>, <0 0 0>, <0 0 0>, <0 0 0>, <0 0 0>]
    trailBone {boneName=cape_05_r}
    trailRelaxationSpeed {editorCurveData {keys [<at 0 15 RCIM_Linear>,<at 0.5289 13.1838 RCIM_Linear>,<at 1 6 RCIM_Linear>]}}
   }
   n31 = Get isIdleFrontEnd() -> (isIdleFrontEnd:bool)
   n32 = "Transform (Modify) Bone - Bone: cape_01_l"!(ComponentPose=@n30, bAlphaBoolEnabled:bool=@n31, Translation="0,0,0", Rotation="15,10,0", Scale="1,1,1") -> (Pose) {
    alphaInputType=Bool
    rotationMode=BMM_Additive
    rotationSpace=BCS_BoneSpace
    alphaBoolBlend {blendOption=Linear}
    boneToModify {boneName=cape_01_l}
    rotation {pitch=15, yaw=10}
   }
   n33 = "Transform (Modify) Bone - Bone: cape_mid_01_l"!(ComponentPose=@n32, bAlphaBoolEnabled:bool=@n31, Translation="0,0,0", Rotation="10,0,0", Scale="1,1,1") -> (Pose) {
    alphaInputType=Bool
    rotationMode=BMM_Additive
    rotationSpace=BCS_BoneSpace
    alphaBoolBlend {blendOption=Linear}
    boneToModify {boneName=cape_mid_01_l}
    rotation {pitch=10}
   }
   n34 = "Transform (Modify) Bone - Bone: cape_01_r"!(ComponentPose=@n33, bAlphaBoolEnabled:bool=@n31, Translation="0,0,0", Rotation="0,10,0", Scale="1,1,1") -> (Pose) {
    alphaInputType=Bool
    rotationMode=BMM_Additive
    rotationSpace=BCS_BoneSpace
    alphaBoolBlend {blendOption=Linear}
    boneToModify {boneName=cape_01_r}
    rotation {yaw=10}
   }
   n35 = "Transform (Modify) Bone - Bone: cape_mid_01_r"!(ComponentPose=@n34, bAlphaBoolEnabled:bool=@n31, Translation="0,0,0", Rotation="5,0,0", Scale="1,1,1") -> (Pose) {
    alphaInputType=Bool
    rotationMode=BMM_Additive
    rotationSpace=BCS_BoneSpace
    alphaBoolBlend {blendOption=Linear}
    boneToModify {boneName=cape_mid_01_r}
    rotation {pitch=5}
   }
   Save cached pose 'trailCache'(Pose=Component To Local(ComponentPose=@n35)) {cacheName=trailCache}
   n36 = "Transform (Modify) Bone - Bone: cape_01_l"!(
    ComponentPose=Local To Component(LocalPose=Use cached pose 'Start_Cache'())
    Alpha:float=1
    Translation="0,0,0"
    Rotation="5,0,0"
    Scale="1,1,1"
   ) -> (Pose) {
    rotationMode=BMM_Additive
    rotationSpace=BCS_BoneSpace
    alphaBoolBlend {blendOption=Linear}
    boneToModify {boneName=cape_01_l}
    rotation {pitch=5}
   }
   n37 = "Transform (Modify) Bone - Bone: cape_03_l"(ComponentPose=@n36, Alpha=1, Translation="0,0,0", Rotation="-10,0,0", Scale="1,1,1") {
    rotationMode=BMM_Additive
    rotationSpace=BCS_BoneSpace
    alphaBoolBlend {blendOption=Linear}
    boneToModify {boneName=cape_03_l}
    rotation {pitch=-10}
   }
   n38 = Trail controller - Bone: cape_mid_06_l!(ComponentPose=@n37, Alpha=1, RelaxationSpeedScale:float=2, FakeVelocity:Vector="0,0,0") -> (Pose) {
    bActorSpaceFakeVel=true
    bShowTrailLocation=true
    bUsePlanarLimit=true
    chainLength=5
    debugLifeTime=0.1
    maxDeltaTime=0.0333
    relaxationSpeedScale=1.2
    alphaScaleBiasClamp {bInterpResult=true, interpSpeedDecreasing=25, interpSpeedIncreasing=25}
    baseJoint {boneName=backpack_root}
    rotationLimits [{limitMax=<180 180 180>}, {limitMax=<180 180 180>}, {limitMax=<180 180 180>}, {limitMax=<180 180 180>}, {limitMax=<180 180 180>}]
    rotationOffsets [<0 0 0>, <0 0 0>, <0 0 0>, <0 0 0>, <0 0 0>]
    trailBone {boneName=cape_mid_06_l}
    trailRelaxationSpeed {editorCurveData {keys [<at 0 15 RCIM_Linear>,<at 1 6 RCIM_Linear>]}}
   }
   n39 = Trail controller - Bone: cape_mid_06_r!(ComponentPose=@n38, Alpha=1, RelaxationSpeedScale:float=2, FakeVelocity:Vector="0,0,0") -> (Pose) {
    bActorSpaceFakeVel=true
    bShowTrailLocation=true
    bUsePlanarLimit=true
    chainLength=5
    debugLifeTime=0.1
    maxDeltaTime=0.0333
    relaxationSpeedScale=1.2
    alphaScaleBiasClamp {bInterpResult=true, interpSpeedDecreasing=25, interpSpeedIncreasing=25}
    baseJoint {boneName=backpack_root}
    rotationLimits [{limitMax=<180 180 180>}, {limitMax=<180 180 180>}, {limitMax=<180 180 180>}, {limitMax=<180 180 180>}, {limitMax=<180 180 180>}]
    rotationOffsets [<0 0 0>, <0 0 0>, <0 0 0>, <0 0 0>, <0 0 0>]
    trailBone {boneName=cape_mid_06_r}
    trailRelaxationSpeed {editorCurveData {keys [<at 0 15 RCIM_Linear>,<at 1 6 RCIM_Linear>]}}
   }
   n40 = "Constraint - cape_r_MSTR (P)"(ComponentPose=@n39, Alpha=1, ConstraintWeights_0=1) {
    alphaBoolBlend {blendOption=Linear}
    boneToModify {boneName=cape_r_MSTR}
    constraintSetup [{targetBone {boneName=upperarm_twist_01_r}}]
    constraintWeights [1]
   }
   n41 = "Constraint - cape_mid_r_MSTR (PP)"!(ComponentPose=@n40, bAlphaBoolEnabled:bool=True, ConstraintWeights_0=0.75, ConstraintWeights_1:float=0.25) -> (Pose) {
    alphaInputType=Bool
    alphaBoolBlend {blendOption=Linear}
    boneToModify {boneName=cape_mid_r_MSTR}
    constraintSetup [{targetBone {boneName=cape_r_MSTR}}, {targetBone {boneName=spine_05}}]
    constraintWeights [1,1]
   }
   n42 = RigidBody!(ComponentPose=@n41, Alpha, OverrideWorldGravity:Vector="0,0,-980", ExternalForce) -> (Pose) {
    bClampLinearTranslationLimitToRefPose=true
    bOverrideWorldGravity=true
    overridePhysicsAsset=&/M_MED_Mystic_Collar_perf_Physics
    simulationSpace=BaseBoneSpace
    alphaBoolBlend {blendOption=Linear}
    baseBoneRef {boneName=pelvis}
    componentAppliedLinearAccClamp=<3000 3000 3000>
    componentLinearAccScale=<0.2 0.2 0.5>
    componentLinearVelScale=<0.2 0.2 0.5>
    overrideWorldGravity {z=-980}
    simSpaceSettings {
     maxAngularAcceleration=50
     maxAngularVelocity=1
     maxLinearAcceleration=300
     maxLinearVelocity=150
     velocityScaleZ=0
     worldAlpha=1
     externalLinearDragV=<0.2 0.2>
    }
   }
   Save cached pose 'dbnoCache'(Pose=Component To Local(ComponentPose=@n42)) {cacheName=dbnoCache}
   n43 = RigidBody!(ComponentPose=Local To Component(LocalPose=Use cached pose 'bodyCache'()), Alpha, ExternalForce) -> (Pose) {
    bClampLinearTranslationLimitToRefPose=true
    bOverrideWorldGravity=true
    overridePhysicsAsset=&/M_MED_Mystic_Body_Physics
    simulationSpace=BaseBoneSpace
    alphaBoolBlend {blendOption=Linear}
    baseBoneRef {boneName=spine_01}
    componentAppliedLinearAccClamp=<1000 1000 1000>
    componentLinearAccScale=<0.01 0.01 0.05>
    componentLinearVelScale=<0.01 0.01 0.05>
    overrideWorldGravity {z=-980}
   }
   n44 = "Transform (Modify) Bone - Bone: dyn_clothBelt_02"(ComponentPose=@n43, Alpha=1, Translation="0,0,0", Rotation="-1.908,-0.191,5.561", Scale="1,1,1") {
    rotationMode=BMM_Additive
    rotationSpace=BCS_BoneSpace
    alphaBoolBlend {blendOption=Linear}
    boneToModify {boneName=dyn_clothBelt_02}
    rotation=<rot -1.9079 -0.191 5.561>
   }
   n45 = RigidBody(ComponentPose=Local To Component(LocalPose=Use cached pose 'bodyCache'())) {
    bClampLinearTranslationLimitToRefPose=true
    bOverrideWorldGravity=true
    overridePhysicsAsset=&/M_MED_Mystic_Body_Crouching_Physics
    simulationSpace=BaseBoneSpace
    alphaBoolBlend {blendOption=Linear}
    baseBoneRef {boneName=root}
    componentAppliedLinearAccClamp=<6000 6000 6000>
    componentLinearAccScale=<0.2 0.2 0.2>
    componentLinearVelScale=<0.2 0.2 0.2>
    overrideWorldGravity {z=-980}
   }
   Save cached pose 'FinalBodyCache'(Pose=Blend Poses by bool(BlendPose_0=Component To Local(ComponentPose=@n45), BlendPose_1=Component To Local(ComponentPose=@n44), bActiveValue=Get bIsCrouching()) {blendType=Linear}) {cacheName=FinalBodyCache}
   n46 = "Constraint - cape_01_l (PP)"!(
    ComponentPose=Local To Component(LocalPose=Use cached pose 'FinalBodyCache'())
    Alpha=1
    ConstraintWeights_0=0.5
    ConstraintWeights_1:float=0.5
   ) -> (Pose) {
    alphaBoolBlend {blendOption=Linear}
    boneToModify {boneName=cape_01_l}
    constraintSetup [{targetBone {boneName=cape_mid_01_l}}, {targetBone {boneName=upperarm_l}}]
    constraintWeights [1,1]
   }
   n47 = "Constraint - cape_mid_01_l (P)"(ComponentPose=@n46, Alpha=1, ConstraintWeights_0=1) {
    alphaBoolBlend {blendOption=Linear}
    boneToModify {boneName=cape_mid_01_l}
    constraintSetup [{targetBone {boneName=spine_05}}]
    constraintWeights [1]
   }
   n48 = "Constraint - cape_mid_01_r (P)"(ComponentPose=@n47, Alpha=1, ConstraintWeights_0=1) {
    alphaBoolBlend {blendOption=Linear}
    boneToModify {boneName=cape_mid_01_r}
    constraintSetup [{targetBone {boneName=spine_05}}]
    constraintWeights [1]
   }
   n49 = Get consumable()
   n50 = "Transform (Modify) Bone - Bone: clavicle_cape_r"(ComponentPose=@n48, bAlphaBoolEnabled=@n49, Translation="0,5,0", Rotation="0.006,15.000,0.011", Scale="1,1,1") {
    alphaInputType=Bool
    rotationMode=BMM_Additive
    rotationSpace=BCS_BoneSpace
    translationMode=BMM_Additive
    translationSpace=BCS_BoneSpace
    alphaBoolBlend {blendInTime=0.1, blendOption=Linear, blendOutTime=0.1}
    boneToModify {boneName=clavicle_cape_r}
    rotation=<rot 0.0059 15 0.0111>
    translation {y=5}
   }
   n51 = "Transform (Modify) Bone - Bone: cape_mid_r_MSTR"!(ComponentPose=@n50, bAlphaBoolEnabled:bool=@n49, Translation="2.780,4.982,0.439", Rotation="-10,5.386,0", Scale="1,1,1") -> (Pose) {
    alphaInputType=Bool
    rotationMode=BMM_Additive
    rotationSpace=BCS_BoneSpace
    translationMode=BMM_Additive
    translationSpace=BCS_BoneSpace
    alphaBoolBlend {blendInTime=0.1, blendOption=Linear, blendOutTime=0.1}
    boneToModify {boneName=cape_mid_r_MSTR}
    rotation {pitch=-10, yaw=5.386}
    translation=<2.78 4.982 0.439>
   }
   n52 = "Transform (Modify) Bone - Bone: cape_r_MSTR"(ComponentPose=@n51, bAlphaBoolEnabled=@n49, Translation="2,3,0", Rotation="0,-29,0", Scale="1,1,1") {
    alphaInputType=Bool
    rotationMode=BMM_Additive
    rotationSpace=BCS_BoneSpace
    translationMode=BMM_Additive
    translationSpace=BCS_BoneSpace
    alphaBoolBlend {blendInTime=0.1, blendOption=Linear, blendOutTime=0.1}
    boneToModify {boneName=cape_r_MSTR}
    rotation {yaw=-29}
    translation=<2 3>
   }
   n53 = "Transform (Modify) Bone - Bone: cape_mid_l_MSTR"!(ComponentPose=@n52, bAlphaBoolEnabled:bool=@n2, Translation="0,0,3", Rotation="5,0,0", Scale="1,1,1") -> (Pose) {
    alphaInputType=Bool
    rotationMode=BMM_Additive
    rotationSpace=BCS_BoneSpace
    translationMode=BMM_Additive
    translationSpace=BCS_BoneSpace
    alphaBoolBlend {blendInTime=0.1, blendOption=Linear, blendOutTime=0.1}
    boneToModify {boneName=cape_mid_l_MSTR}
    rotation {pitch=5}
    translation {z=3}
   }
   n54 = RigidBody!(ComponentPose=@n53, Alpha, OverrideWorldGravity:Vector=Get GravityOverride(), ExternalForce) -> (Pose) {
    bClampLinearTranslationLimitToRefPose=true
    bOverrideWorldGravity=true
    overridePhysicsAsset=&/M_MED_Mystic_Collar_BAckpack_perf_Physics
    simulationSpace=BaseBoneSpace
    alphaBoolBlend {blendOption=Linear}
    baseBoneRef {boneName=root}
    componentAppliedLinearAccClamp=<1000 1000 1000>
    componentLinearAccScale {z=0.3}
    componentLinearVelScale {z=0.3}
    overrideWorldGravity {z=-980}
    simSpaceSettings {
     maxAngularAcceleration=50
     maxAngularVelocity=1
     maxLinearAcceleration=300
     maxLinearVelocity=150
     velocityScaleZ=0
     worldAlpha=1
     externalLinearDragV=<0.2 0.2>
    }
   }
   n55 = Component To Local(ComponentPose=@n54)
   n56 = Get consumable()
   n57 = "Transform (Modify) Bone - Bone: clavicle_cape_r"(
    ComponentPose=Local To Component(LocalPose=Use cached pose 'FinalBodyCache'())
    bAlphaBoolEnabled=@n56
    Translation="0,5,0"
    Rotation="0.006,15.000,0.011"
    Scale="1,1,1"
   ) {
    alphaInputType=Bool
    rotationMode=BMM_Additive
    rotationSpace=BCS_BoneSpace
    translationMode=BMM_Additive
    translationSpace=BCS_BoneSpace
    alphaBoolBlend {blendInTime=0.1, blendOption=Linear, blendOutTime=0.1}
    boneToModify {boneName=clavicle_cape_r}
    rotation=<rot 0.0059 15 0.0111>
    translation {y=5}
   }
   n58 = "Transform (Modify) Bone - Bone: cape_mid_r_MSTR"(ComponentPose=@n57, bAlphaBoolEnabled=@n56, Translation="2.780,4.982,0.439", Rotation="-10,5.386,0", Scale="1,1,1") {
    alphaInputType=Bool
    rotationMode=BMM_Additive
    rotationSpace=BCS_BoneSpace
    translationMode=BMM_Additive
    translationSpace=BCS_BoneSpace
    alphaBoolBlend {blendInTime=0.1, blendOption=Linear, blendOutTime=0.1}
    boneToModify {boneName=cape_mid_r_MSTR}
    rotation {pitch=-10, yaw=5.386}
    translation=<2.78 4.982 0.439>
   }
   n59 = "Transform (Modify) Bone - Bone: cape_r_MSTR"(ComponentPose=@n58, bAlphaBoolEnabled=@n56, Translation="2,3,0", Rotation="0,-29,0", Scale="1,1,1") {
    alphaInputType=Bool
    rotationMode=BMM_Additive
    rotationSpace=BCS_BoneSpace
    translationMode=BMM_Additive
    translationSpace=BCS_BoneSpace
    alphaBoolBlend {blendInTime=0.1, blendOption=Linear, blendOutTime=0.1}
    boneToModify {boneName=cape_r_MSTR}
    rotation {yaw=-29}
    translation=<2 3>
   }
   n60 = Get bIsMoving2D()
   n61 = "Transform (Modify) Bone - Bone: cape_mid_l_MSTR"(ComponentPose=@n59, bAlphaBoolEnabled=@n60, Translation="0,0,5", Rotation="15,0,0", Scale="1,1,1") {
    alphaInputType=Bool
    rotationMode=BMM_Additive
    rotationSpace=BCS_BoneSpace
    translationMode=BMM_Additive
    translationSpace=BCS_BoneSpace
    alphaBoolBlend {blendInTime=0.1, blendOption=Linear, blendOutTime=0.1}
    boneToModify {boneName=cape_mid_l_MSTR}
    rotation {pitch=15}
    translation {z=5}
   }
   n62 = "Transform (Modify) Bone - Bone: cape_mid_r_MSTR"(ComponentPose=@n61, bAlphaBoolEnabled=@n60, Translation="0,0,5", Rotation="15,0,0", Scale="1,1,1") {
    alphaInputType=Bool
    rotationMode=BMM_Additive
    rotationSpace=BCS_BoneSpace
    translationMode=BMM_Additive
    translationSpace=BCS_BoneSpace
    alphaBoolBlend {blendInTime=0.1, blendOption=Linear, blendOutTime=0.1}
    boneToModify {boneName=cape_mid_r_MSTR}
    rotation {pitch=15}
    translation {z=5}
   }
   n63 = RigidBody(ComponentPose=@n62, OverrideWorldGravity=Get GravityOverride_cape() -> (GravityOverride_cape:Vector)) {
    bClampLinearTranslationLimitToRefPose=true
    bOverrideWorldGravity=true
    overridePhysicsAsset=&M_MED_Mystic_Collar_perf_Physics
    simulationSpace=BaseBoneSpace
    alphaBoolBlend {blendOption=Linear}
    baseBoneRef {boneName=root}
    componentAppliedLinearAccClamp=<3000 3000 3000>
    componentLinearAccScale=<0.2 0.2 0.5>
    componentLinearVelScale=<0.2 0.2 0.5>
    overrideWorldGravity {z=-980}
    simSpaceSettings {
     maxAngularAcceleration=50
     maxAngularVelocity=1
     maxLinearAcceleration=300
     maxLinearVelocity=150
     velocityScaleZ=0
     worldAlpha=1
     externalLinearDragV=<0.2 0.2>
    }
   }
   n64 = Component To Local(ComponentPose=@n63)
   n65 = Get Is Wearing Backbling()
   n66 = RigidBody!(
    ComponentPose=Local To Component(LocalPose=Use cached pose 'FinalBodyCache'())
    bAlphaBoolEnabled:bool
    OverrideWorldGravity=Get GravityOverride()
    ExternalForce
   ) -> (Pose) {
    alphaInputType=Bool
    bClampLinearTranslationLimitToRefPose=true
    bOverrideWorldGravity=true
    overridePhysicsAsset=&/M_MED_Mystic_Collar_Crouching_BackblingPhysics
    simulationSpace=BaseBoneSpace
    alphaBoolBlend {blendOption=Linear}
    baseBoneRef {boneName=root}
    componentAppliedLinearAccClamp=<3000 3000 3000>
    componentLinearAccScale {z=0.5}
    componentLinearVelScale {z=0.5}
    overrideWorldGravity {z=-980}
    simSpaceSettings {
     maxAngularAcceleration=50
     maxAngularVelocity=1
     maxLinearAcceleration=300
     maxLinearVelocity=150
     velocityScaleZ=0
     worldAlpha=1
     externalLinearDragV=<0.2 0.2>
    }
   }
   n67 = "Transform (Modify) Bone - Bone: cape_mid_01_r"(ComponentPose=@n66, bAlphaBoolEnabled=True, Translation="0.008,7.316,3.287", Rotation="0,0,0", Scale="1,1,1") {
    alphaInputType=Bool
    translationMode=BMM_Additive
    translationSpace=BCS_BoneSpace
    alphaBoolBlend {blendOption=Linear}
    boneToModify {boneName=cape_mid_01_r}
    translation=<0.008 7.316 3.287>
   }
   Save cached pose 'RbanCache'(Pose=Blend Poses by bool(
     BlendPose_0=Component To Local(ComponentPose=@n67)
     BlendPose_1=Blend Poses by bool(BlendPose_0=@n55, BlendPose_1=@n64, BlendTime_0=0.2, BlendTime_1=0.5, bActiveValue=@n65) {blendType=Linear}
     BlendTime_0=0.2
     BlendTime_1=0.5
     bActiveValue=Get bIsCrouching()
    ) {blendType=Linear}) {cacheName=RbanCache}
   n68 = "Transform (Modify) Bone - Bone: dyn_root_slope"(
    ComponentPose=Local To Component(LocalPose=Copy Pose From Mesh() {bCopyCurves=true, bUseAttachedParent=true})
    Alpha=Get PlummetAlpha() -> (PlummetAlpha:double)
    Translation="0,0,-600"
    Rotation="0,0,0"
    Scale="1,1,1"
   ) {
    translationMode=BMM_Additive
    translationSpace=BCS_BoneSpace
    alphaBoolBlend {blendOption=Linear}
    boneToModify {boneName=dyn_root_slope}
    translation {z=-600}
   }
   n69 = Control Rig(Source=Component To Local(ComponentPose=@n68)) {
    alphaBoolBlend {blendOption=Linear}
    controlRigAssetReference {blueprintRigClass=&/M_MED_Mystic_Deform_CtrlRig_twists.M_MED_Mystic_Deform_CtrlRig_twists_C}
   }
   n70 = Control Rig!(
    Source=@n69
    Alpha
    KneeRot_fr_l_114deg:Vector
    KneeTrans_fr_l_114deg:Vector="0.258,-3.591,0.009"
    KneeScale_fr_l_114deg:Vector="1,1,1"
    KneeScale_fr_l_130deg:Vector="1,1,1"
    KneeRot_fr_l_130deg:Vector="0,0,-51.2"
    KneeTrans_fr_l_130deg:Vector="0.258,-3.591,0.009"
    isRbanEnabled:bool=Get bRigidBodyEnabled()
   ) -> (Pose) {
    alphaBoolBlend {blendOption=Linear}
    controlRigAssetReference {blueprintRigClass=&/M_MED_Mystic_Deform_CtrlRig_sculpters.M_MED_Mystic_Deform_CtrlRig_sculpters_C}
   }
   n71 = Anim Dynamics - Bone: dyn_ring_01(ComponentPose=Local To Component(LocalPose=@n70), Alpha=1, ExternalForce="0,0,0") {
    angularDampingOverride=1
    angularSpringConstant=20000
    bAngularSpring=true
    bOverrideAngularDamping=true
    bUseGravityOverride=true
    gravityScale=0.35
    simulationSpace=RootRelative
    alphaBoolBlend {blendOption=Linear}
    boundBone {boneName=dyn_ring_01}
    componentLinearAccScale=<1 1 1>
    componentLinearVelScale=<1 1 1>
    gravityOverride {z=-980}
    physicsBodyDefinitions [{
      boundBone {boneName=None}
      boxExtents=<2 5 5>
      constraintSetup {angularLimitsMax=<0 0 0>, angularLimitsMin=<0 0 0>, angularTarget=<0 0 0>}
      localJointOffset=<2 0 0>
     }]
   }
   n72 = Get bIsCrouching()
   n73 = "Transform (Modify) Bone - Bone: dyn_rapel_back_r_02"!(ComponentPose=@n71, bAlphaBoolEnabled:bool=@n72, Translation="-4,0,0", Rotation="0,0,0", Scale="0.7,1,1") -> (Pose) {
    alphaInputType=Bool
    scaleMode=BMM_Additive
    translationMode=BMM_Additive
    translationSpace=BCS_BoneSpace
    alphaBoolBlend {blendOption=Linear}
    boneToModify {boneName=dyn_rapel_back_r_02}
    scale {x=0.7}
    translation {x=-4}
   }
   n74 = "Transform (Modify) Bone - Bone: dyn_lapel_front_l_MSTR"!(ComponentPose=@n73, bAlphaBoolEnabled:bool=@n72, Translation="0,0,0", Rotation="0.004,7.326,0.008", Scale="1,1,1") -> (Pose) {
    alphaInputType=Bool
    rotationMode=BMM_Additive
    rotationSpace=BCS_BoneSpace
    scaleMode=BMM_Additive
    translationMode=BMM_Additive
    translationSpace=BCS_BoneSpace
    alphaBoolBlend {blendOption=Linear}
    boneToModify {boneName=dyn_lapel_front_l_MSTR}
    rotation=<rot 0.004 7.326 0.008>
   }
   n75 = Get isCrouchingFloat() -> (isCrouchingFloat:double)
   n76 = "Transform (Modify) Bone - Bone: dyn_clothBelt_01"(ComponentPose=@n74, Alpha=@n75, Translation="-6.104,-1.572,-2.383", Rotation="88.326,-73.884,-95.134", Scale="1,1,1") {
    rotationMode=BMM_Additive
    rotationSpace=BCS_BoneSpace
    scaleMode=BMM_Additive
    translationMode=BMM_Additive
    translationSpace=BCS_BoneSpace
    alphaBoolBlend {blendOption=Linear}
    alphaScaleBiasClamp {bInterpResult=true}
    boneToModify {boneName=dyn_clothBelt_01}
    rotation=<rot 88.326 -73.884 -95.134>
    translation=<-6.104 -1.572 -2.383>
   }
   n77 = "Transform (Modify) Bone - Bone: dyn_clothBelt_02"(ComponentPose=@n76, Alpha=@n75, Translation="-3.215,-5.224,-0.081", Rotation="-56.671,-66.493,8.715", Scale="1,1,1") {
    rotationMode=BMM_Additive
    rotationSpace=BCS_BoneSpace
    scaleMode=BMM_Additive
    translationMode=BMM_Additive
    translationSpace=BCS_BoneSpace
    alphaBoolBlend {blendOption=Linear}
    alphaScaleBiasClamp {bInterpResult=true}
    boneToModify {boneName=dyn_clothBelt_02}
    rotation=<rot -56.671 -66.493 8.715>
    translation=<-3.215 -5.224 -0.081>
   }
   n78 = "Transform (Modify) Bone - Bone: dyn_clothBelt_03"(ComponentPose=@n77, Alpha=@n75, Translation="1.186,3.050,0.095", Rotation="-1.057,-15.038,2.235", Scale="1,1,1") {
    rotationMode=BMM_Additive
    rotationSpace=BCS_BoneSpace
    scaleMode=BMM_Additive
    translationMode=BMM_Additive
    translationSpace=BCS_BoneSpace
    alphaBoolBlend {blendOption=Linear}
    alphaScaleBiasClamp {bInterpResult=true}
    boneToModify {boneName=dyn_clothBelt_03}
    rotation=<rot -1.057 -15.038 2.235>
    translation=<1.186 3.05 0.095>
   }
   n79 = Get isCrouchingFloat()
   n80 = "Transform (Modify) Bone - Bone: dyn_belt_01"(ComponentPose=@n78, Alpha=@n79, Translation="-1.238,-1.957,-0.133", Rotation="46.542,-100.147,-120.050", Scale="1,1,1") {
    rotationMode=BMM_Additive
    rotationSpace=BCS_BoneSpace
    scaleMode=BMM_Additive
    translationMode=BMM_Additive
    translationSpace=BCS_BoneSpace
    alphaBoolBlend {blendOption=Linear}
    alphaScaleBiasClamp {bInterpResult=true}
    boneToModify {boneName=dyn_belt_01}
    rotation=<rot 46.542 -100.147 -120.05>
    translation=<-1.238 -1.957 -0.133>
   }
   n81 = "Transform (Modify) Bone - Bone: dyn_belt_02"(ComponentPose=@n80, Alpha=@n79, Translation="0.278,0.217,0.613", Rotation="-23.070,-30.418,16.568", Scale="1,1,1") {
    rotationMode=BMM_Additive
    rotationSpace=BCS_BoneSpace
    scaleMode=BMM_Additive
    translationMode=BMM_Additive
    translationSpace=BCS_BoneSpace
    alphaBoolBlend {blendOption=Linear}
    alphaScaleBiasClamp {bInterpResult=true}
    boneToModify {boneName=dyn_belt_02}
    rotation=<rot -23.07 -30.418 16.568>
    translation=<0.278 0.217 0.613>
   }
   n82 = "Transform (Modify) Bone - Bone: dyn_belt_03"!(ComponentPose=@n81, Alpha:float=@n79, Translation="0,0,0", Rotation="-32.411,3.172,-36.609", Scale="1,1,1") -> (Pose) {
    rotationMode=BMM_Additive
    rotationSpace=BCS_BoneSpace
    scaleMode=BMM_Additive
    translationMode=BMM_Additive
    translationSpace=BCS_BoneSpace
    alphaBoolBlend {blendOption=Linear}
    alphaScaleBiasClamp {bInterpResult=true}
    boneToModify {boneName=dyn_belt_03}
    rotation=<rot -32.411 3.172 -36.6091>
   }
   n83 = "Transform (Modify) Bone - Bone: dyn_belt_05"(ComponentPose=@n82, Alpha=@n79, Translation="0,0,0", Rotation="-4.027,-2.360,-3.115", Scale="1,1,1") {
    rotationMode=BMM_Additive
    rotationSpace=BCS_BoneSpace
    scaleMode=BMM_Additive
    translationMode=BMM_Additive
    translationSpace=BCS_BoneSpace
    alphaBoolBlend {blendOption=Linear}
    alphaScaleBiasClamp {bInterpResult=true}
    boneToModify {boneName=dyn_belt_05}
    rotation=<rot -4.027 -2.36 -3.115>
   }
   n84 = "Transform (Modify) Bone - Bone: dyn_belt_04"(ComponentPose=@n83, Alpha=@n79, Translation="0,0,0", Rotation="0,0,0", Scale="1,1,1") {
    rotationMode=BMM_Additive
    rotationSpace=BCS_BoneSpace
    scaleMode=BMM_Additive
    translationMode=BMM_Additive
    translationSpace=BCS_BoneSpace
    alphaBoolBlend {blendOption=Linear}
    alphaScaleBiasClamp {bInterpResult=true}
    boneToModify {boneName=dyn_belt_04}
   }
   n85 = "Transform (Modify) Bone - Bone: dyn_clothBelt_01"(ComponentPose=@n84, Alpha=0, Translation="-2,-8,0", Rotation="-0.000,0.000,50", Scale="1,1,1") {
    rotationMode=BMM_Additive
    translationMode=BMM_Additive
    alphaBoolBlend {blendOption=Linear}
    alphaScaleBiasClamp {bInterpResult=true}
    boneToModify {boneName=dyn_clothBelt_01}
    rotation {roll=50}
    translation=<-2 -8>
   }
   n86 = "Transform (Modify) Bone - Bone: dyn_belt_01"(ComponentPose=@n85, Alpha=0, Translation="-4,-8,0", Rotation="-10,0,60", Scale="1,1,1") {
    rotationMode=BMM_Additive
    translationMode=BMM_Additive
    translationSpace=BCS_BoneSpace
    alphaBoolBlend {blendOption=Linear}
    alphaScaleBiasClamp {bInterpResult=true}
    boneToModify {boneName=dyn_belt_01}
    rotation {pitch=-10, roll=60}
    translation=<-4 -8>
   }
   n87 = "Transform (Modify) Bone - Bone: dyn_belt_02"(ComponentPose=@n86, Alpha=0, Translation="-1,0,0", Rotation="-10,0,40", Scale="1,1,1") {
    rotationMode=BMM_Additive
    translationMode=BMM_Additive
    translationSpace=BCS_BoneSpace
    alphaBoolBlend {blendOption=Linear}
    boneToModify {boneName=dyn_belt_02}
    rotation {pitch=-10, roll=40}
    translation {x=-1}
   }
   n88 = "Transform (Modify) Bone - Bone: dyn_clothBelt_01"(ComponentPose=@n87, Alpha=0, Translation="-2,-4,5", Rotation="0,0,-20", Scale="1,1,1") {
    rotationMode=BMM_Additive
    translationMode=BMM_Additive
    alphaBoolBlend {blendOption=Linear}
    boneToModify {boneName=dyn_clothBelt_01}
    rotation {roll=-20}
    translation=<-2 -4 5>
   }
   n89 = "Transform (Modify) Bone - Bone: dyn_belt_01"(ComponentPose=@n88, Alpha=0, Translation="0,-1,7", Rotation="0,0,-10", Scale="1,1,1") {
    rotationMode=BMM_Additive
    translationMode=BMM_Additive
    translationSpace=BCS_BoneSpace
    alphaBoolBlend {blendOption=Linear}
    boneToModify {boneName=dyn_belt_01}
    rotation {roll=-10}
    translation {y=-1, z=7}
   }
   n90 = "Transform (Modify) Bone - Bone: dyn_belt_02"(ComponentPose=@n89, Alpha=0, Translation="0,0,0", Rotation="0,0,0", Scale="1,1,1") {
    rotationMode=BMM_Additive
    translationMode=BMM_Additive
    translationSpace=BCS_BoneSpace
    alphaBoolBlend {blendOption=Linear}
    boneToModify {boneName=dyn_belt_02}
   }
   n91 = "Constraint - dyn_clothBelt_01 (PP)"!(ComponentPose=@n90, Alpha=@n1, ConstraintWeights_0=0.5, ConstraintWeights_1:float=0.5) -> (Pose) {
    alphaBoolBlend {blendOption=Linear}
    boneToModify {boneName=dyn_clothBelt_01}
    constraintSetup [{targetBone {boneName=thigh_l}}, {targetBone {boneName=thigh_twist_01_l}}]
    constraintWeights [1,1]
   }
   n92 = "Constraint - dyn_clothBelt_02 (PP)"!(ComponentPose=@n91, Alpha=@n1, ConstraintWeights_0=0.5, ConstraintWeights_1:float=0.5) -> (Pose) {
    alphaBoolBlend {blendOption=Linear}
    boneToModify {boneName=dyn_clothBelt_02}
    constraintSetup [{targetBone {boneName=thigh_l}}, {targetBone {boneName=thigh_twist_01_l}}]
    constraintWeights [1,1]
   }
   n93 = Look At - Bone: dyn_belt_MSTR(ComponentPose=@n92, Alpha=@n1) {
    alphaBoolBlend {blendOption=Linear}
    boneToModify {boneName=dyn_belt_MSTR}
    lookAtLocation=<0 24.0435 0>
    lookAtTarget {boneReference {boneName=calf_l}}
    lookAt_Axis {axis=<1 0>}
   }
   n94 = "Constraint - dyn_belt_02 (RR)"!(ComponentPose=@n93, Alpha=@n1, ConstraintWeights_0=0.5, ConstraintWeights_1:float=0.5) -> (Pose) {
    alphaBoolBlend {blendOption=Linear}
    boneToModify {boneName=dyn_belt_02}
    constraintSetup [{transformType=Rotation, targetBone {boneName=thigh_l}}, {transformType=Rotation, targetBone {boneName=thigh_twist_01_l}}]
    constraintWeights [1,1]
   }
   n95 = "Transform (Modify) Bone - Bone: dyn_clothBelt_01"(ComponentPose=@n94, Alpha=1, Translation="0,0,0", Rotation="5.000,0.000,-0.000", Scale="1,1,1") {
    rotationMode=BMM_Additive
    rotationSpace=BCS_BoneSpace
    alphaBoolBlend {blendOption=Linear}
    boneToModify {boneName=dyn_clothBelt_01}
    rotation=<rot 5 0 0>
   }
   n96 = "Transform (Modify) Bone - Bone: dyn_clothBelt_02"(ComponentPose=@n95, Alpha=1, Translation="0,0,0", Rotation="5.000,0.000,-0.000", Scale="1,1,1") {
    rotationMode=BMM_Additive
    rotationSpace=BCS_BoneSpace
    alphaBoolBlend {blendOption=Linear}
    boneToModify {boneName=dyn_clothBelt_02}
    rotation=<rot 5 0 0>
   }
   n97 = "Transform (Modify) Bone - Bone: dyn_lapel_back_l_01"!(ComponentPose=@n96, bAlphaBoolEnabled:bool=True, Translation="0,0,0", Rotation="0,15,0", Scale="1,1,1") -> (Pose) {
    alphaInputType=Bool
    rotationMode=BMM_Additive
    rotationSpace=BCS_BoneSpace
    alphaBoolBlend {blendOption=Linear}
    boneToModify {boneName=dyn_lapel_back_l_01}
    rotation {yaw=15}
   }
   n98 = "Transform (Modify) Bone - Bone: dyn_rapel_back_r_01"!(ComponentPose=@n97, bAlphaBoolEnabled:bool=True, Translation="0,0,0", Rotation="0,-30,0", Scale="1,1,1") -> (Pose) {
    alphaInputType=Bool
    rotationMode=BMM_Additive
    rotationSpace=BCS_BoneSpace
    alphaBoolBlend {blendOption=Linear}
    boneToModify {boneName=dyn_rapel_back_r_01}
    rotation {yaw=-30}
   }
   n99 = "Constraint - dyn_collarA_l_01 (P)"(ComponentPose=@n98, Alpha=0.4, ConstraintWeights_0=1) {
    alphaBoolBlend {blendOption=Linear}
    boneToModify {boneName=dyn_collarA_l_01}
    constraintSetup [{targetBone {boneName=clavicle_l}}]
    constraintWeights [1]
   }
   n100 = "Constraint - dyn_collarA_l_03 (P)"(ComponentPose=@n99, Alpha=0.7, ConstraintWeights_0=1) {
    alphaBoolBlend {blendOption=Linear}
    boneToModify {boneName=dyn_collarA_l_03}
    constraintSetup [{targetBone {boneName=clavicle_l}}]
    constraintWeights [1]
   }
   n101 = "Constraint - dyn_collarA_r_02 (P)"(ComponentPose=@n100, Alpha=0.4, ConstraintWeights_0=1) {
    alphaBoolBlend {blendOption=Linear}
    boneToModify {boneName=dyn_collarA_r_02}
    constraintSetup [{targetBone {boneName=clavicle_r}}]
    constraintWeights [1]
   }
   n102 = Get LookUp() -> (LookUp:double)
   n103 = "Transform (Modify) Bone - Bone: dyn_collar_c_01"(ComponentPose=@n101, Alpha=@n102, Translation="-1.5,1.1,-1.7", Rotation="-76.560,36.868,-52.961", Scale="1,1,1") {
    rotationMode=BMM_Additive
    rotationSpace=BCS_BoneSpace
    translationMode=BMM_Additive
    translationSpace=BCS_BoneSpace
    alphaBoolBlend {blendOption=Linear}
    alphaScaleBiasClamp {bInterpResult=true}
    boneToModify {boneName=dyn_collar_c_01}
    rotation=<rot -76.56 36.8681 -52.9611>
    translation=<-1.5 1.1 -1.7>
   }
   n104 = "Transform (Modify) Bone - Bone: dyn_collarB_r_01"(ComponentPose=@n103, Alpha=@n102, Translation="-1.095,1.580,-0.534", Rotation="-85,0.043,-4.265", Scale="1,1,1") {
    rotationMode=BMM_Additive
    rotationSpace=BCS_BoneSpace
    translationMode=BMM_Additive
    translationSpace=BCS_BoneSpace
    alphaBoolBlend {blendOption=Linear}
    alphaScaleBiasClamp {bInterpResult=true}
    boneToModify {boneName=dyn_collarB_r_01}
    rotation=<rot -85 0.043 -4.265>
    translation=<-1.095 1.58 -0.534>
   }
   n105 = "Transform (Modify) Bone - Bone: dyn_collarB_l_01"(ComponentPose=@n104, Alpha=@n102, Translation="0,0,0", Rotation="-49.945,-11.835,2.238", Scale="1,1,1") {
    rotationMode=BMM_Additive
    rotationSpace=BCS_BoneSpace
    translationMode=BMM_Additive
    translationSpace=BCS_BoneSpace
    alphaBoolBlend {blendOption=Linear}
    alphaScaleBiasClamp {bInterpResult=true}
    boneToModify {boneName=dyn_collarB_l_01}
    rotation=<rot -49.945 -11.835 2.238>
   }
   n106 = Get look Up ADS() -> (look Up ADS:double)
   n107 = "Transform (Modify) Bone - Bone: dyn_collar_c_01"(ComponentPose=@n105, Alpha=@n106, Translation="-1.5,1.1,0", Rotation="-57.875,-23.265,-20.411", Scale="1,1,1") {
    rotationMode=BMM_Additive
    rotationSpace=BCS_BoneSpace
    translationMode=BMM_Replace
    translationSpace=BCS_BoneSpace
    alphaBoolBlend {blendOption=Linear}
    alphaScaleBiasClamp {bInterpResult=true, interpSpeedDecreasing=9}
    boneToModify {boneName=dyn_collar_c_01}
    rotation=<rot -57.8751 -23.2651 -20.411>
    translation=<-1.5 1.1>
   }
   n108 = "Transform (Modify) Bone - Bone: dyn_collarB_r_01"(ComponentPose=@n107, Alpha=@n106, Translation="2,0,3", Rotation="-80,-3,-9", Scale="1,1,1") {
    rotationMode=BMM_Additive
    rotationSpace=BCS_BoneSpace
    translationMode=BMM_Additive
    translationSpace=BCS_BoneSpace
    alphaBoolBlend {blendOption=Linear}
    alphaScaleBiasClamp {bInterpResult=true, interpSpeedDecreasing=9}
    boneToModify {boneName=dyn_collarB_r_01}
    rotation=<rot -80 -3 -9>
    translation {x=2, z=3}
   }
   n109 = "Transform (Modify) Bone - Bone: dyn_collarB_l_01"(ComponentPose=@n108, Alpha=@n106, Translation="0,0,0", Rotation="-35.000,0.026,-0.035", Scale="0,1,1") {
    rotationMode=BMM_Additive
    rotationSpace=BCS_BoneSpace
    translationMode=BMM_Additive
    translationSpace=BCS_BoneSpace
    alphaBoolBlend {blendOption=Linear}
    alphaScaleBiasClamp {bInterpResult=true, interpSpeedDecreasing=9}
    boneToModify {boneName=dyn_collarB_l_01}
    rotation=<rot -35 0.026 -0.035>
    scale {x=0}
   }
   n110 = Anim Dynamics - Bone: dyn_collarA_l_02(ComponentPose=@n109, Alpha=0.5, ExternalForce="0,0,0") {
    angularDampingOverride=0.999
    angularSpringConstant=500000
    bAngularSpring=true
    bOverrideAngularDamping=true
    bUseGravityOverride=true
    gravityScale=0.35
    simulationSpace=RootRelative
    alphaBoolBlend {blendOption=Linear}
    boundBone {boneName=dyn_collarA_l_02}
    componentAppliedLinearAccClamp=<10000 10000 10000>
    componentLinearAccScale=<0.2 0.2 1>
    componentLinearVelScale=<0.2 0.2 1>
    gravityOverride {z=-980}
    physicsBodyDefinitions [{
      boundBone {boneName=dyn_collarA_l_02}
      boxExtents=<1 3 3>
      constraintSetup {angularTargetAxis=AxisY, angularLimitsMax=<0 10 0>, angularLimitsMin=<0 -10 0>, angularTarget=<0 0 0>}
      localJointOffset=<1 0 0>
     }]
   }
   n111 = Anim Dynamics - Bone: dyn_collarA_r_02(ComponentPose=@n110, Alpha=0.5, ExternalForce="0,0,0") {
    angularDampingOverride=0.999
    angularSpringConstant=500000
    bAngularSpring=true
    bOverrideAngularDamping=true
    bUseGravityOverride=true
    gravityScale=0.35
    simulationSpace=RootRelative
    alphaBoolBlend {blendOption=Linear}
    boundBone {boneName=dyn_collarA_r_02}
    componentAppliedLinearAccClamp=<10000 10000 10000>
    componentLinearAccScale=<0.2 0.2 1>
    componentLinearVelScale=<0.2 0.2 1>
    gravityOverride {z=-980}
    physicsBodyDefinitions [{
      boundBone {boneName=dyn_collarA_r_02}
      boxExtents=<1 3 3>
      constraintSetup {angularTargetAxis=AxisY, angularLimitsMax=<0 10 0>, angularLimitsMin=<0 -10 0>, angularTarget=<0 0 0>}
      localJointOffset=<-1 0 0>
     }]
   }
   n112 = Get head rot right() -> (head rot right:double)
   n113 = "Transform (Modify) Bone - Bone: dyn_collar_c_01"(ComponentPose=@n111, Alpha=@n112, Translation="0,0,0", Rotation="-45,-0.000,-0.000", Scale="1,1,1") {
    rotationMode=BMM_Additive
    rotationSpace=BCS_BoneSpace
    translationMode=BMM_Additive
    translationSpace=BCS_BoneSpace
    alphaBoolBlend {blendOption=Linear}
    alphaScaleBiasClamp {bInterpResult=true, interpSpeedDecreasing=9}
    boneToModify {boneName=dyn_collar_c_01}
    rotation {pitch=-45}
   }
   n114 = "Transform (Modify) Bone - Bone: dyn_collarA_r_01"(ComponentPose=@n113, Alpha=@n112, Translation="2,0,2", Rotation="-75,-0.000,-0.000", Scale="1,1,1") {
    rotationMode=BMM_Additive
    rotationSpace=BCS_BoneSpace
    translationMode=BMM_Additive
    translationSpace=BCS_BoneSpace
    alphaBoolBlend {blendOption=Linear}
    alphaScaleBiasClamp {bInterpResult=true, interpSpeedDecreasing=9}
    boneToModify {boneName=dyn_collarA_r_01}
    rotation {pitch=-75}
    translation {x=2, z=2}
   }
   n115 = "Transform (Modify) Bone - Bone: dyn_collarA_r_02"(ComponentPose=@n114, Alpha=@n112, Translation="0,0,0", Rotation="95,-0.000,-0.001", Scale="1,1,1") {
    rotationMode=BMM_Additive
    rotationSpace=BCS_BoneSpace
    translationMode=BMM_Additive
    translationSpace=BCS_BoneSpace
    alphaBoolBlend {blendOption=Linear}
    alphaScaleBiasClamp {bInterpResult=true, interpSpeedDecreasing=9}
    boneToModify {boneName=dyn_collarA_r_02}
    rotation {pitch=95, roll=-0.001}
   }
   n116 = "Transform (Modify) Bone - Bone: dyn_collarB_r_01"(ComponentPose=@n115, Alpha=@n112, Translation="2.5,0.000,2", Rotation="-60,-0.000,-0.000", Scale="1,1,1") {
    rotationMode=BMM_Additive
    rotationSpace=BCS_BoneSpace
    translationMode=BMM_Additive
    translationSpace=BCS_BoneSpace
    alphaBoolBlend {blendOption=Linear}
    alphaScaleBiasClamp {bInterpResult=true, interpSpeedDecreasing=9}
    boneToModify {boneName=dyn_collarB_r_01}
    rotation {pitch=-60}
    translation {x=2.5, z=2}
   }
   n117 = Get head rot left() -> (head rot left:double)
   n118 = "Transform (Modify) Bone - Bone: dyn_collarA_l_01"(ComponentPose=@n116, Alpha=@n117, Translation="-1.748,-0.000,-1.741", Rotation="-60,0,0", Scale="1,1,1") {
    rotationMode=BMM_Additive
    rotationSpace=BCS_BoneSpace
    translationMode=BMM_Additive
    translationSpace=BCS_BoneSpace
    alphaBoolBlend {blendOption=Linear}
    alphaScaleBiasClamp {bInterpResult=true, interpSpeedDecreasing=9}
    boneToModify {boneName=dyn_collarA_l_01}
    rotation {pitch=-60}
    translation {x=-1.748, z=-1.741}
   }
   n119 = "Transform (Modify) Bone - Bone: dyn_collarA_l_02"(ComponentPose=@n118, Alpha=@n117, Translation="0,0,0", Rotation="66.686,0,0", Scale="1,1,1") {
    rotationMode=BMM_Additive
    rotationSpace=BCS_BoneSpace
    translationMode=BMM_Additive
    translationSpace=BCS_BoneSpace
    alphaBoolBlend {blendOption=Linear}
    alphaScaleBiasClamp {bInterpResult=true, interpSpeedDecreasing=9}
    boneToModify {boneName=dyn_collarA_l_02}
    rotation {pitch=66.686}
   }
   n120 = "Transform (Modify) Bone - Bone: dyn_collarB_l_01"(ComponentPose=@n119, Alpha=@n117, Translation="-1.063,-0.003,0", Rotation="-60.000,-0.000,-0.000", Scale="1,1,1") {
    rotationMode=BMM_Additive
    rotationSpace=BCS_BoneSpace
    translationMode=BMM_Additive
    translationSpace=BCS_BoneSpace
    alphaBoolBlend {blendOption=Linear}
    alphaScaleBiasClamp {bInterpResult=true, interpSpeedDecreasing=9}
    boneToModify {boneName=dyn_collarB_l_01}
    rotation=<rot -60 0 0>
    translation=<-1.063 -0.003>
   }
   n121 = Get pistol() -> (pistol:bool)
   n122 = Get melee() -> (melee:bool)
   n123 = AND Boolean(A=NOT Boolean(A=Get bIsParachuteOpened()), B=NOT Boolean(A=Get bIsFalling()))
   n124 = "Constraint - clavicle_cape_l (P)"!(ComponentPose=@n120, bAlphaBoolEnabled:bool=AND Boolean(A=OR Boolean(A=@n121, B=@n122), B=@n123), ConstraintWeights_0=1) -> (Pose) {
    alphaInputType=Bool
    alphaBoolBlend {blendInTime=0.5, blendOption=Linear, blendOutTime=0.5}
    boneToModify {boneName=clavicle_cape_l}
    constraintSetup [{targetBone {boneName=clavicle_l}}]
    constraintWeights [1]
   }
   n125 = Get pistol()
   n126 = Get melee()
   n127 = Get consumable()
   n128 = "Constraint - clavicle_cape_r (P)"!(
    ComponentPose=@n124
    bAlphaBoolEnabled:bool=AND Boolean(A=OR Boolean!(A=@n125, B=@n126, C:bool=@n127) -> (ReturnValue), B=@n123)
    ConstraintWeights_0=1
   ) -> (Pose) {
    alphaInputType=Bool
    alphaBoolBlend {blendInTime=0.5, blendOption=Linear, blendOutTime=0.5}
    boneToModify {boneName=clavicle_cape_r}
    constraintSetup [{targetBone {boneName=clavicle_r}}]
    constraintWeights [1]
   }
   n129 = RigidBody!(ComponentPose=@n128, Alpha:float=0, ExternalForce) -> (Pose) {
    bClampLinearTranslationLimitToRefPose=true
    bOverrideWorldGravity=true
    overridePhysicsAsset=&/M_MED_Mystic_Necklace_Physics
    simulationSpace=BaseBoneSpace
    alphaBoolBlend {blendOption=Linear}
    baseBoneRef {boneName=spine_05}
    componentAppliedLinearAccClamp=<6000 6000 6000>
    componentLinearAccScale=<0.05 0.05 0.05>
    componentLinearVelScale=<0.05 0.05 0.05>
    overrideWorldGravity {z=-980}
   }
   n130 = Get armL_rot up() -> (armL_rot up:double)
   n131 = "Constraint - clavicle_cape_l (PP)"!(ComponentPose=@n129, Alpha=@n130, ConstraintWeights_0=0.5, ConstraintWeights_1:float=0.5) -> (Pose) {
    alphaBoolBlend {blendInTime=0.1, blendOption=Linear, blendOutTime=0.1}
    boneToModify {boneName=clavicle_cape_l}
    constraintSetup [{targetBone {boneName=upperarm_twist_01_l}}, {targetBone {boneName=clavicle_l}}]
    constraintWeights [1,1]
   }
   n132 = "Constraint - cape_mid_l_MSTR (PP)"!(ComponentPose=@n131, Alpha=@n130, ConstraintWeights_0=0.5, ConstraintWeights_1:float=0.5) -> (Pose) {
    alphaBoolBlend {blendOption=Linear}
    boneToModify {boneName=cape_mid_l_MSTR}
    constraintSetup [{targetBone {boneName=cape_l_MSTR}}, {targetBone {boneName=spine_05}}]
    constraintWeights [1,1]
   }
   n133 = "Transform (Modify) Bone - Bone: cape_l_MSTR"!(
    ComponentPose=@n132
    bAlphaBoolEnabled:bool=NOT Boolean(A=Get bRigidBodyEnabled())
    Translation="0.000,-1.982,6.616"
    Rotation="0,0,0"
    Scale="1,1,1"
   ) -> (Pose) {
    alphaInputType=Bool
    translationMode=BMM_Additive
    translationSpace=BCS_BoneSpace
    alphaBoolBlend {blendInTime=0.1, blendOption=Linear, blendOutTime=0.1}
    boneToModify {boneName=cape_l_MSTR}
    translation {y=-1.982, z=6.616}
   }
   n134 = Get arm_L_rot() -> (arm_L_rot:double)
   n135 = "Transform (Modify) Bone - Bone: cape_l_MSTR"!(ComponentPose=@n133, Alpha:float=@n134, Translation="1,0,3", Rotation="15,0,-0.000", Scale="1,1,1") -> (Pose) {
    rotationMode=BMM_Additive
    rotationSpace=BCS_BoneSpace
    translationMode=BMM_Additive
    translationSpace=BCS_BoneSpace
    alphaBoolBlend {blendOption=Linear}
    boneToModify {boneName=cape_l_MSTR}
    rotation {pitch=15}
    translation {x=1, z=3}
   }
   n136 = "Transform (Modify) Bone - Bone: clavicle_cape_l"!(ComponentPose=@n135, Alpha:float=@n134, Translation="0,0,0", Rotation="5.000,-15,-0.000", Scale="1,1,1") -> (Pose) {
    rotationMode=BMM_Additive
    rotationSpace=BCS_BoneSpace
    alphaBoolBlend {blendOption=Linear}
    boneToModify {boneName=clavicle_cape_l}
    rotation {pitch=5, yaw=-15}
   }
   n137 = Get arm_r_rot() -> (arm_r_rot:double)
   n138 = "Transform (Modify) Bone - Bone: cape_r_MSTR"!(ComponentPose=@n136, Alpha:float=@n137, Translation="0,0,0", Rotation="10,-0.000,-0.000", Scale="1,1,1") -> (Pose) {
    rotationMode=BMM_Additive
    rotationSpace=BCS_BoneSpace
    alphaBoolBlend {blendOption=Linear}
    boneToModify {boneName=cape_r_MSTR}
    rotation {pitch=10}
   }
   n139 = "Transform (Modify) Bone - Bone: clavicle_cape_r"!(ComponentPose=@n138, Alpha:float=@n137, Translation="0,0,0", Rotation="5.000,20,0.001", Scale="1,1,1") -> (Pose) {
    rotationMode=BMM_Additive
    rotationSpace=BCS_BoneSpace
    alphaBoolBlend {blendOption=Linear}
    boneToModify {boneName=clavicle_cape_r}
    rotation=<rot 5 20 0.001>
   }
   n140 = OR Boolean!(A=Get bIsSkydiveDiveMode(), B=Get bIsSkydiveFloating()) -> (ReturnValue)
   n141 = "Constraint - clavicle_cape_l (PP)"!(ComponentPose=@n139, bAlphaBoolEnabled:bool=@n140, ConstraintWeights_0=1, ConstraintWeights_1=0) -> (Pose) {
    alphaInputType=Bool
    alphaBoolBlend {blendInTime=0.1, blendOption=Linear, blendOutTime=0.1}
    boneToModify {boneName=clavicle_cape_l}
    constraintSetup [{targetBone {boneName=upperarm_l}}, {targetBone {boneName=clavicle_l}}]
    constraintWeights [1,1]
   }
   n142 = "Constraint - clavicle_cape_r (PP)"!(ComponentPose=@n141, bAlphaBoolEnabled:bool=@n140, ConstraintWeights_0=1, ConstraintWeights_1:float=0) -> (Pose) {
    alphaInputType=Bool
    alphaBoolBlend {blendInTime=0.1, blendOption=Linear, blendOutTime=0.1}
    boneToModify {boneName=clavicle_cape_r}
    constraintSetup [{targetBone {boneName=upperarm_r}}, {targetBone {boneName=spine_05}}]
    constraintWeights [1,1]
   }
   n143 = "Transform (Modify) Bone - Bone: cape_l_MSTR"!(ComponentPose=@n142, bAlphaBoolEnabled:bool=@n140, Translation="3,0,5", Rotation="-0.000,15,-0.000", Scale="1,1,1") -> (Pose) {
    alphaInputType=Bool
    rotationMode=BMM_Additive
    rotationSpace=BCS_BoneSpace
    translationMode=BMM_Additive
    translationSpace=BCS_BoneSpace
    alphaBoolBlend {blendInTime=0.1, blendOption=Linear, blendOutTime=0.1}
    boneToModify {boneName=cape_l_MSTR}
    rotation {yaw=15}
    translation {x=3, z=5}
   }
   n144 = "Transform (Modify) Bone - Bone: cape_r_MSTR"!(ComponentPose=@n143, bAlphaBoolEnabled:bool=@n140, Translation="0.000,0,10", Rotation="0,0,0", Scale="1,1,1") -> (Pose) {
    alphaInputType=Bool
    rotationMode=BMM_Additive
    rotationSpace=BCS_BoneSpace
    translationMode=BMM_Additive
    translationSpace=BCS_BoneSpace
    alphaBoolBlend {blendInTime=0.1, blendOption=Linear, blendOutTime=0.1}
    boneToModify {boneName=cape_r_MSTR}
    translation {z=10}
   }
   n145 = "Transform (Modify) Bone - Bone: cape_mid_l_MSTR"(ComponentPose=@n144, bAlphaBoolEnabled=@n140, Translation="0,0,5", Rotation="15,15,0", Scale="1,1,1") {
    alphaInputType=Bool
    rotationMode=BMM_Additive
    rotationSpace=BCS_BoneSpace
    translationMode=BMM_Additive
    translationSpace=BCS_BoneSpace
    alphaBoolBlend {blendInTime=0.1, blendOption=Linear, blendOutTime=0.1}
    boneToModify {boneName=cape_mid_l_MSTR}
    rotation {pitch=15, yaw=15}
    translation {z=5}
   }
   n146 = "Transform (Modify) Bone - Bone: cape_mid_r_MSTR"(ComponentPose=@n145, bAlphaBoolEnabled=@n140, Translation="0,0,5", Rotation="15,0,0", Scale="1,1,1") {
    alphaInputType=Bool
    rotationMode=BMM_Additive
    rotationSpace=BCS_BoneSpace
    translationMode=BMM_Additive
    translationSpace=BCS_BoneSpace
    alphaBoolBlend {blendInTime=0.1, blendOption=Linear, blendOutTime=0.1}
    boneToModify {boneName=cape_mid_r_MSTR}
    rotation {pitch=15}
    translation {z=5}
   }
   n147 = "Transform (Modify) Bone - Bone: dyn_root_slope"!(
    ComponentPose=@n146
    bAlphaBoolEnabled:bool=OR Boolean(A=Get bIsFalling(), B=Get bIsDBNO())
    Translation="0,0,-300"
    Rotation="0,0,0"
    Scale="1,1,1"
   ) -> (Pose) {
    alphaInputType=Bool
    rotationMode=BMM_Additive
    rotationSpace=BCS_BoneSpace
    translationMode=BMM_Additive
    translationSpace=BCS_BoneSpace
    alphaBoolBlend {blendInTime=0.1, blendOption=Linear, blendOutTime=1}
    boneToModify {boneName=dyn_root_slope}
    translation {z=-300}
   }
   Save cached pose 'bodyCache'(Pose=Component To Local(ComponentPose=@n147)) {cacheName=bodyCache}
   Use cached pose 'FinalBodyCache'()
   Get bClothEnabled()
   Get Thigh L Rot Pos() -> (Thigh L Rot Pos:double)
   Get bClothEnabled()
   Get GravityOverride()
   Get Thigh L Rot Neg() -> (Thigh L Rot Neg:double)
   Get bIsSkyDiving()
   Get head tilt right() -> (head tilt right:double)
  ]}
 EventGraph {nodes [
   n1 = Cloth Controller Function DEPRECATED!(
    execute=@n4.then
    CharacterPartAnimInstance:Custom Character Part Anim Instance Object Reference=Self-Reference()
    ClothingAssetName:String=M_MED_Mystic_Clothing_1
    Cloth Paramaters:Cloth Params Struct Structure=Get Cloth Paramaters() -> (Cloth Paramaters:Cloth Params Struct Structure)
    Cloth Core Settings:Cloth Core Setting Struct Structure=Get Cloth Core Settings() -> (Cloth Core Settings:Cloth Core Setting Struct Structure)
    ClothUpAxis:Vector="1,0,0"
    OnGround_Moving_SpeedLimit:double=2
    OnGround_Moving_AngularVel_Max:double=0.05
    OnGround_Moving_FictitiousVel_Max:double=0.6
    LinVelScale_FwdBwd_Clamp:double=0
    LinVelScale_UpDwn_Clamp:double=0.7
    DeltaTime:double=Get DeltaTimeLocal() -> (DeltaTimeLocal:double)
    GameTimeInSeconds:double=GetGameTimeInSeconds()
    WindGustsParams:Wind Gust Params Struct Structure=Get Wind Gusts Params() -> (Wind Gusts Params:Wind Gust Params Struct Structure)
    WindGustLocalSettings:Wind Gust Local Setting Struct Structure=Get Gust Local Settings() -> (Gust Local Settings:Wind Gust Local Setting Struct Structure)
   ) -> (
    then
    CurrentStateOut:String
    Cloth Settings Out:Cloth Setting Struct Structure
    Cloth Core Settings Out:Cloth Core Setting Struct Structure
    WindGustSettings Out:Wind Gust Setting Struct Structure
    WindGustLocalSettings Out:Wind Gust Local Setting Struct Structure
    GustStrength - WindVelocityOut:Vector
    LinearVelVector-DELETE LATER:Vector
   )
   DrawVectorFromJointFunction!(
    execute
    SkelMeshComponent:Scene Component Object Reference
    JointName:Name=head
    LineEnd:Vector=@n1.GustStrength - WindVelocityOut
    JointOffset:Vector="0,0,30"
    LineColor:Linear Color Structure=(R=1,G=0,B=0,A=1)
    Thickness:double=1
    Foreground:EDrawDebugSceneDepthPriorityGroup Enum=Foreground
   ) -> (then)
   > DrawJointDebugFunction!(
    execute
    SkelMeshComponent:Scene Component Object Reference
    JointOffset:Vector="0, 0, 0"
    JointName:Array of Names=Make Array!("[0]":Name=root, "[1]":Name=pelvis) -> (Array:Array of Names)
    Scale:double=10
    Duration:double=0
    Thickness:double=0.7
    DebugDraw:bool=true
    Foreground:EDrawDebugSceneDepthPriorityGroup Enum=Foreground
   ) -> (then)
   > DrawVectorFromJointFunction(
    JointName=head
    LineEnd=Break Cloth Setting Struct!(ClothSettingStruct:"Cloth Setting Struct Structure (by ref)"=@n1.Cloth Settings Out) -> (
     LinearVelocityScale_7_#1:Vector=~
     AngularVelocityScale_6_#2:float=~
     FictitiousAngularScale_10_#3:float=~
     GravityOverride_15_#4:Vector=~
     DragCoefficient_23_#5:Vector 2D Structure
     LiftCoefficient_24_#6:Vector 2D Structure
     JointName_27_#7:Name
    ) {is=title K2Node_BreakStruct}.GravityOverride_15_#4
    JointOffset="0, 0,40"
    LineColor=(R=0.0192,G=0,B=1,A=1)
    Thickness=2
    Foreground=Foreground
   )
   n2 = Event BlueprintUpdateAnimation()
   > Set DeltaTimeLocal(DeltaTimeLocal:double=@n2.DeltaTimeX) -> (Output_Get:double)
   > n3 = Sequence!(execute) -> (
    then_0
    then_1
    then_2:Exec
    then_3:Exec
    then_4:Exec
    then_5:Exec
    then_6:Exec
    then_7:Exec
    then_8:Exec
    then_9:Exec
    then_10:Exec
    then_11:Exec
    then_12:Exec
    then_13:Exec
    then_14:Exec
    then_15:Exec
    then_16:Exec
   )
   > n4 = Branch(Condition=Get bClothEnabled())
   n5 = Gravity Override State Function DEPRECATED(
    execute=@n4.else
    CharacterPartAnimInstance=Self-Reference()
    GameTimeInSeconds=GetGameTimeInSeconds()
    GravityOverrideParams=Get Gravity Override Params()
   )
   > n6 = Set GravityOverride(GravityOverride=@n5)
   > n7 = Gravity Override State Function DEPRECATED(
    CharacterPartAnimInstance=Self-Reference()
    GameTimeInSeconds=GetGameTimeInSeconds()
    GravityOverrideParams=Get Gravity Override Params_cape() -> (Gravity Override Params_cape:Gravity Override Params Struct Structure)
   )
   > Set GravityOverride_cape(GravityOverride_cape:Vector=@n7) -> (Output_Get:Vector)
   n8 = PlayerMovement(execute=@n3.then_1, AnimInstance=Self-Reference(), XDotProductCoefficient=1, YDotProductCoefficient=1, ZDotProductCoefficient=1)
   n9 = "Absolute (Float)"(A=@n8)
   n10 = MapRangeClamped(Value=@n8.YMovement, InRangeA=0, InRangeB=1)
   > Set BwdMovement(BwdMovement:double=MapRangeClamped(Value=float + float(A=@n9, B=@n10) {is=title K2Node_CommutativeAssociativeBinaryOperator}, InRangeA=0, InRangeB=1.4)) -> (Output_Get:double)
   Set Is Wearing Backbling(execute=@n3.then_10, Is Wearing Backbling=CheckForBackbling(self=Self-Reference()))
   n11 = Branch(execute=@n3.then_11, Condition=Get bIsSkyDiving())
   Set arm_L_rot(execute=[@n11.then, @n12.else], arm_L_rot:double=0) -> (Output_Get:double)
   > Set arm_r_rot(arm_r_rot:double=0) -> (Output_Get:double)
   n12 = Branch(execute=@n11.else, Condition=OR Boolean(A=Get melee(), B=Get pistol()))
   n13 = Get Skel_Mesh()
   > Set arm_L_rot(arm_L_rot=GetAlphaFromReferenceTransformDelta(SkeletalMeshComponent=@n13, BoneName=upperarm_l, BaseName=spine_05, InputRange=(X=10,Y=-20), TwistAxis="1,0,0"))
   > Set arm_r_rot(arm_r_rot=GetAlphaFromReferenceTransformDelta(SkeletalMeshComponent=@n13, BoneName=upperarm_r, BaseName=spine_05, InputRange=(X=10,Y=-20), TwistAxis="1,0,0"))
   n14 = Branch(execute=@n3.then_13, Condition=Get bIsSkyDiving())
   > Set armL_rot up(armL_rot up:double=0) -> (Output_Get:double)
   > Set arm_L_rot(arm_L_rot=0)
   n15 = GetAlphaFromReferenceTransformDelta(
    SkeletalMeshComponent=Get Skel_Mesh()
    BoneName=upperarm_l
    BaseName=spine_05
    InputRange=(X=60,Y=100)
    OutputRange=(X=1,Y=0.4)
    TwistAxis="1,0,0"
   )
   Set armL_rot up(execute=@n14.else, armL_rot up=@n15)
   Get!(execute:Exec=@n3.then_15) -> (then:Exec, else:Exec, Anim_Inst:Anim Instance Object Reference)
   n16 = GetCurveValue(self=Get Anim_Inst() -> (Anim_Inst:Anim Instance Object Reference), CurveName=PhysicsGroundOffset)
   > n17 = Branch(Condition="float > float"(A=@n16))
   > Set PlummetAlpha(PlummetAlpha:double=@n16) -> (Output_Get:double)
   Set PlummetAlpha(execute=@n17.else, PlummetAlpha=0)
   n18 = Branch(execute=@n3.then_2, Condition=Get bIsSwimSprinting() -> (bIsSwimSprinting:bool))
   > Set TrailControllerRot(TrailControllerRot:Rotator="5,0, 0") -> (Output_Get:Rotator)
   n19 = Branch(execute=@n18.else, Condition=Get bIsSkyDiving())
   > n20 = Branch(Condition=Get bIsParachuteOpened())
   > Set TrailControllerRot(TrailControllerRot="20,0, 0")
   n21 = Branch(execute=@n20.else, Condition="float > float"(A=Get PawnVelocityZ(), B=50) {is=title K2Node_CallFunction})
   > Set BwdMovement(BwdMovement=float * float(A=Get BwdMovement(), B=0))
   > Set TrailControllerRot(TrailControllerRot="45,0, 0")
   n22 = Branch(execute=@n21.else, Condition=Get bIsSkydiveDiveMode())
   > Set TrailControllerRot(TrailControllerRot="15,0, 0")
   Set TrailControllerRot(execute=@n22.else, TrailControllerRot="30,0, 0")
   Set TrailControllerRot(execute=@n19.else, TrailControllerRot="5,0, 0")
   > Set BwdMovement(BwdMovement=float * float(A=Get BwdMovement(), B=1))
   n23 = Get Skel_Mesh()
   n24 = BreakTransform(InTransform=GetDeltaTransformFromRefPose(self=@n23, BoneName=upperarm_l, BaseName=clavicle_l))
   Set armL_rotZ(
    execute=@n3.then_3
    armL_rotZ:double=Remap Float(Input=BreakRotator(InRot=@n24.Rotation), OldRange=(X=16,Y=36), NewRange=(X=0,Y=1), Clamp=true) -> (Output=~)
   ) -> (Output_Get:double)
   n25 = GetDeltaTransformFromRefPose(self=@n23, BoneName=upperarm_r, BaseName=clavicle_r)
   > Set armR_rotZ(armR_rotZ:double=BreakRotator(InRot=BreakTransform(InTransform=@n25).Rotation).Yaw) -> (Output_Get:double)
   n26 = Branch(execute=@n3.then_4, Condition=Get bIsCrouching())
   > Set isCrouchingFloat(isCrouchingFloat:double=1) -> (Output_Get:double)
   > Set isNOTCrouchingFloat(isNOTCrouchingFloat:double=0) -> (Output_Get:double)
   Set isCrouchingFloat(execute=@n26.else, isCrouchingFloat=0)
   > Set isNOTCrouchingFloat(isNOTCrouchingFloat=1)
   n27 = Branch(execute=@n3.then_5, Condition=Get bIsCrouching())
   n28 = GetAlphaFromReferenceTransformDelta(SkeletalMeshComponent=Get Skel_Mesh(), BoneName=thigh_l, BaseName=pelvis, InputRange=(X=-150,Y=-30), TwistAxis="1,0,0")
   > Set Thigh L Rot Neg(Thigh L Rot Neg:double=Remap Float(Input=@n28.RawOutput, OldRange=(X=-100,Y=-30), NewRange=(X=0,Y=1), Clamp=true) -> (Output=~)) -> (Output_Get:double)
   Set Thigh L Rot Neg(execute=@n27.else, Thigh L Rot Neg=0)
   n29 = Branch(execute=@n3.then_6, Condition=Get bIsCrouching())
   > Set Thigh L Rot Pos(Thigh L Rot Pos:double=Remap Float(Input=@n28.RawOutput, OldRange=(X=-130,Y=-100), NewRange=(X=1,Y=0), Clamp=true) -> (Output=~)) -> (Output_Get:double)
   Set Thigh L Rot Pos(execute=@n29.else, Thigh L Rot Pos=0)
   n30 = Branch(execute=@n3.then_7, Condition=Get bIsTargeting())
   > Set is Targeting float(is Targeting float:double=1) -> (Output_Get:double)
   Set is Targeting float(execute=@n30.else, is Targeting float=0)
   n31 = GetOwnerSkeletalMeshComponent()
   n32 = BreakTransform(InTransform=GetDeltaTransformFromRefPose(self=@n31, BoneName=head, BaseName=spine_05))
   n33 = Set LookUp(execute=@n3.then_8, LookUp:double=MapRangeClamped(Value=BreakRotator(InRot=@n32.Rotation).Yaw, InRangeA=-10, InRangeB=-60)) -> (Output_Get:double)
   > n34 = Branch(Condition=Get bIsTargeting())
   > Set look Up ADS(look Up ADS:double=@n33) -> (Output_Get:double)
   > Set LookUp(LookUp=0)
   Set look Up ADS(execute=@n34.else, look Up ADS=0)
   n35 = BreakTransform(InTransform=GetDeltaTransformFromRefPose(self=@n31, BoneName=upperarm_l, BaseName=clavicle_l))
   Set LookDown(execute=@n3.then_9, LookDown:double=MapRangeClamped(Value=BreakRotator(InRot=@n35.Rotation).Yaw, InRangeA=50, InRangeB=15)) -> (Output_Get:double)
   n36 = BreakTransform(InTransform=GetDeltaTransformFromRefPose(self=@n31, BoneName=upperarm_r, BaseName=clavicle_r))
   > Set LookUpRifle(LookUpRifle:double=MapRangeClamped(Value=BreakRotator(InRot=@n36.Rotation).Yaw, InRangeA=60, InRangeB=78)) -> (Output_Get:double)
   n37 = Cast To FortPlayerAnimInstance(execute=@n3.then_12, Object=GetAnimInstance(self=GetOwnerSkeletalMeshComponent()))
   > Set pistol(pistol:bool=false) -> (Output_Get:bool)
   > Set rifle(rifle:bool=false) -> (Output_Get:bool)
   > Set melee(melee:bool=false) -> (Output_Get:bool)
   > Set consumable(consumable:bool=false) -> (Output_Get:bool)
   > n38 = Switch on EFortWeaponCoreAnimation!(execute, Selection:EFortWeaponCoreAnimation Enum=Get WeaponCoreAnim(self=@n37)) -> (
    Melee:Exec
    Pistol:Exec
    Shotgun:Exec
    PaperBlueprint:Exec
    Rifle:Exec
    MeleeOneHand:Exec
    MachinePistol:Exec
    RocketLauncher:Exec
    GrenadeLauncher:Exec
    GoingCommando:Exec
    AssaultRifle:Exec
    TacticalShotgun:Exec
    SniperRifle:Exec
    TrapPlacement:Exec
    ShoulderLauncher:Exec
    AbilityDecoTool:Exec
    Crossbow:Exec
    C4:Exec
    RemoteControl:Exec
    DualWield:Exec
    AR_BullPup:Exec
    AR_ForwardGrip:Exec
    MedPackPaddles:Exec
    SMG_P90:Exec
    AR_DrumGun:Exec
    Consumable_Small:Exec
    Consumable_Large:Exec
    Balloon:Exec
    MountedTurret:Exec
    CreativeTool:Exec
    ExplosiveBow:Exec
    AshtonIndigo:Exec
    AshtonChicago:Exec
    MeleeDualWield:Exec
    Unarmed:Exec
   )
   Branch(execute=[
     @n38.AssaultRifle
     @n38.Crossbow
     @n38.GrenadeLauncher
     @n38.Rifle
     @n38.RocketLauncher
     @n38.Shotgun
     @n38.SniperRifle
     @n38.TacticalShotgun
    ])
   > Set rifle(rifle=true)
   Branch(execute=[@n38.Consumable_Large, @n38.Consumable_Small])
   > Set consumable(consumable=true)
   n39 = Get bIsMoving2D()
   Branch(execute=[@n38.MachinePistol, @n38.Pistol], Condition=@n39)
   > Set pistol(pistol=true)
   Branch(execute=@n38.Melee, Condition=@n39)
   > Set melee(melee=true)
   n40 = GetAnimInstance(self=GetOwnerSkeletalMeshComponent())
   Is Valid(exec=@n3.then_14, InputObject=@n40)
   > n41 = Branch(Condition=Montage_IsPlaying(self=@n40, Montage=/BRCosmetics/Animation/Game/MainPlayer/Emotes/JellyFrog/Emote_JellyFrog_CMM_M))
   > Set head rot right(head rot right:double=GetAlphaFromReferenceTransformDelta(
     SkeletalMeshComponent=Get Skel_Mesh()
     BoneName=head
     BaseName=spine_05
     DeltaComponentType=EulerX
     InputRange=(X=-20,Y=-60)
     TwistAxis="1,0,0"
    )) -> (Output_Get:double)
   n42 = Get Skel_Mesh()
   > Set head rot left(head rot left:double=GetAlphaFromReferenceTransformDelta(SkeletalMeshComponent=@n42, BoneName=head, BaseName=spine_05, DeltaComponentType=EulerX, InputRange=(X=20,Y=50), TwistAxis="1,0,0")) -> (Output_Get:double)
   > Set head tilt right(head tilt right:double=GetAlphaFromReferenceTransformDelta(SkeletalMeshComponent=@n42, BoneName=head, BaseName=spine_05, DeltaComponentType=EulerY, InputRange=(X=5,Y=20), TwistAxis="1,0,0")) -> (Output_Get:double)
   Set head rot right(execute=@n41.else, head rot right=0)
   > Set head rot left(head rot left=0)
   > Set head tilt right(head tilt right=0)
   n43 = GetAnimInstance(self=GetOwnerSkeletalMeshComponent())
   Is Valid(exec=@n3.then_16, InputObject=@n43)
   > n44 = Branch(Condition=Montage_IsPlaying(self=@n43, Montage=/BRCosmetics/Animation/Game/MainPlayer/Menu/BR/CustomSkins/_s20/MysticM/Mystic_Male_Idle_Montage))
   > Set isIdleFrontEnd(isIdleFrontEnd:bool=true) -> (Output_Get:bool)
   Set isIdleFrontEnd(execute=@n44.else, isIdleFrontEnd=false)
   Event OnParachuteOpened()
   > n45 = Delay(Duration=1)
   ResetDynamics(execute=[@n45.then, @n49.then, @n50.then], InTeleportType=TeleportPhysics)
   > ResetDynamics(InTeleportType=ResetPhysics)
   Event BlueprintInitializeAnimation()
   n46 = GetOwnerSkeletalMeshComponent()
   > Is Valid(InputObject=@n46)
   > Set Skel_Mesh(Skel_Mesh=@n46)
   > n47 = Cast To FortPlayerPawn(Object=GetOwningActor())
   > n48 = Set Pawn(Pawn:Fort Player Pawn Object Reference=@n47) -> (Output_Get:Fort Player Pawn Object Reference)
   > Set Anim_Inst(Anim_Inst:Anim Instance Object Reference=GetAnimInstance(self=Get Mesh(self=@n48))) -> (Output_Get:Anim Instance Object Reference)
   n49 = Event OnBeginSkydiving()
   n50 = Event OnEndSkydiving()
   DrawVectorFromJointFunction(JointName=head, LineEnd=@n6, JointOffset="0, 0, 0", LineColor=(R=0.0192,G=0,B=1,A=1), Thickness=2, Foreground=Foreground)
   n51 = GetDeltaTransformFromRefPose(self=@n31, BoneName=upperarm_r, BaseName=clavicle_r)
   MapRangeClamped(Value=BreakRotator(InRot=BreakTransform(InTransform=@n51).Rotation).Yaw, InRangeA=40, InRangeB=90)
   Get bIsPlayingMeleeAnim()
   Set Gust Local Settings(execute=@n1.then, Gust Local Settings:Wind Gust Local Setting Struct Structure=@n1.WindGustLocalSettings Out) -> (Output_Get:Wind Gust Local Setting Struct Structure)
   Set arm_L_rot(arm_L_rot=@n15)
   Set Alpha Output(Alpha Output:double=GetAlphaFromReferenceTransformDelta(SkeletalMeshComponent=Get Skel_Mesh(), BoneName=upperarm_l, BaseName=spine_05, InputRange=(X=-150,Y=-30), TwistAxis="1,0,0")) -> (Output_Get:double)
  ]}
]
variables [
 Alpha Output {type=number}
 Anim_Inst {type=AnimInstance, default=null}
 arm_L_rot {type=number}
 arm_r_rot {type=number}
 armL_rot up {type=number}
 armL_rotZ {type=number}
 armR_rotZ {type=number}
 BwdMovement {type=number}
 Cloth Core Settings {
  type=ClothCoreSettingStruct
  default {
   collisionThickness=0
   dampingCoefficient=0.1
   frictionCoefficient=0.3
   localDampingCoefficient=0.1
   numberIterations=2
   numberSubsteps=1
   selfCollisionThickness=0
   animDriveDamping=<0 1>
   animDriveStiffness=<0 1>
   areaStiffness=<1 1>
   bendingStiffness=<1 1>
   edgeStiffness=<1 1>
   tetherStiffness=<1 1>
  }
 }
 Cloth Paramaters {
  type=ClothParamsStruct
  default {
   dBNO {
    angularVelocityScale=0.6
    fictitiousAngularScale=0.1
    jointName=pelvis
    dragCoefficient=<0.2 1>
    gravityOverride=<-980 400 0>
    liftCoefficient=<0 0.5>
    linearVelocityScale=<0.3 0.3 0.1>
   }
   dBNO_Carried {
    angularVelocityScale=0.6
    fictitiousAngularScale=0.1
    jointName=pelvis
    dragCoefficient=<0.2 1>
    gravityOverride=<-980 -200 0>
    liftCoefficient=<0 0.5>
    linearVelocityScale=<0.3 0.3 0.1>
   }
   dBNO_Impulse {
    angularVelocityScale=0
    fictitiousAngularScale=0
    jointName=pelvis
    dragCoefficient=<0 0.2>
    gravityOverride=<-3000 -2000 0>
    liftCoefficient=<0 0.1>
    linearVelocityScale=<0.3 0.1 0.3>
   }
   dBNO_RollBwd {
    angularVelocityScale=0
    fictitiousAngularScale=0
    jointName=pelvis
    dragCoefficient=<0 0.2>
    gravityOverride=<-2000 0 0>
    liftCoefficient=<0 0.1>
    linearVelocityScale=<0.3 0.1 0.3>
   }
   dBNO_RollFwd {
    angularVelocityScale=0
    fictitiousAngularScale=0
    jointName=pelvis
    dragCoefficient=<0 0.2>
    gravityOverride=<-980 0 0>
    liftCoefficient=<0 0.1>
    linearVelocityScale=<0.3 0.1 0.3>
   }
   dBNO_Tumbling {
    angularVelocityScale=0
    fictitiousAngularScale=0
    jointName=pelvis
    dragCoefficient=<0 0.2>
    gravityOverride=<-2500 -300 0>
    liftCoefficient=<0 0.1>
    linearVelocityScale=<0.3 0.1 0.3>
   }
   emoteOrMelee {
    angularVelocityScale=0.2
    fictitiousAngularScale=0.1
    jointName=pelvis
    dragCoefficient=<0 0.3>
    gravityOverride=<-980 -500 0>
    liftCoefficient=<0 0.2>
    linearVelocityScale=<0.05 0.05 0.05>
   }
   front_End {
    angularVelocityScale=0.1
    fictitiousAngularScale=0.2
    jointName=root
    dragCoefficient=<0.2 0.2>
    gravityOverride=<0 -200 -980>
    liftCoefficient=<0.1 0.1>
    linearVelocityScale=<0.1 0.1 0.1>
   }
   onGround_Crouch_Moving {
    angularVelocityScale=0.1
    fictitiousAngularScale=0.3
    jointName=root
    dragCoefficient=<2 2>
    gravityOverride=<0 0 -980>
    liftCoefficient=<0.6 0.6>
    linearVelocityScale=<0.6 0.6 0.5>
   }
   onGround_Falling {
    angularVelocityScale=0.6
    fictitiousAngularScale=0
    jointName=root
    dragCoefficient=<0 0.2>
    gravityOverride=<0 -500 -980>
    liftCoefficient=<0 0.1>
    linearVelocityScale=<2 1 0.3>
   }
   onGround_Moving {
    angularVelocityScale=0.6
    fictitiousAngularScale=0.1
    jointName=root
    dragCoefficient=<0 0.3>
    gravityOverride=<0 -100 -980>
    liftCoefficient=<0 0.2>
    linearVelocityScale=<0.6 1 1>
   }
   onGround_Standing {
    angularVelocityScale=0.05
    fictitiousAngularScale=0.6
    jointName=root
    dragCoefficient=<0 2>
    gravityOverride=<0 -100 -980>
    liftCoefficient=<0 0.5>
    linearVelocityScale=<0.6 1 1>
   }
   rail_Grinding {
    angularVelocityScale=0.2
    fictitiousAngularScale=0
    jointName=root
    dragCoefficient=<0.1 0.2>
    gravityOverride=<50 -500 -400>
    liftCoefficient=<0 0.5>
    linearVelocityScale=<0.1 0.1 0.1>
   }
   riding_Moving {
    angularVelocityScale=0.05
    fictitiousAngularScale=1
    jointName=root
    dragCoefficient=<0.5 0.5>
    gravityOverride=<0 0 -980>
    liftCoefficient=<0.5 0.5>
    linearVelocityScale=<0.4 0.6 0.3>
   }
   riding_Moving_Targeting {
    angularVelocityScale=0.05
    fictitiousAngularScale=0.7
    jointName=root
    dragCoefficient=<0.5 0.5>
    gravityOverride=<0 0 -980>
    liftCoefficient=<0.5 0.5>
    linearVelocityScale=<0.2 0.6 0.2>
   }
   riding_Standing {
    angularVelocityScale=0.05
    fictitiousAngularScale=1
    jointName=root
    dragCoefficient=<0.5 0.5>
    gravityOverride=<0 -450 -980>
    liftCoefficient=<0.5 0.5>
    linearVelocityScale=<0.6 0.6 0.3>
   }
   skydive_DivingDown {
    angularVelocityScale=0.1
    fictitiousAngularScale=0
    jointName=pelvis
    dragCoefficient=<0 1.5>
    gravityOverride=<-450 -100 0>
    liftCoefficient=<0 0.3>
    linearVelocityScale=<0.1 0.1 0.1>
   }
   skydive_FloatingDown {
    angularVelocityScale=0.1
    fictitiousAngularScale=0
    jointName=root
    dragCoefficient=<0 2.5>
    gravityOverride=<0 0 980>
    liftCoefficient=<0 0.4>
    linearVelocityScale=<0.2 0.2 0.1>
   }
   skydive_GoingUp {
    angularVelocityScale=0.1
    fictitiousAngularScale=0
    jointName=root
    dragCoefficient=<0 0>
    gravityOverride=<0 -3000 3000>
    liftCoefficient=<0 0>
    linearVelocityScale=<0.1 0.1 0.05>
   }
   skydive_Parachuting {
    angularVelocityScale=0.1
    fictitiousAngularScale=0
    jointName=root
    dragCoefficient=<0 2.5>
    gravityOverride=<0 0 -980>
    liftCoefficient=<0 0.5>
    linearVelocityScale=<0.2 0.2 0.2>
   }
   skydive_Surfing {
    angularVelocityScale=0.6
    fictitiousAngularScale=0.1
    jointName=root
    dragCoefficient=<1.5 1.5>
    gravityOverride=<0 -600 -500>
    liftCoefficient=<1 1>
    linearVelocityScale=<0.1 0.1 0.3>
   }
   swimming {
    angularVelocityScale=0.6
    fictitiousAngularScale=0
    jointName=root
    dragCoefficient=<0 2>
    gravityOverride=<0 -600 400>
    liftCoefficient=<0 0.4>
    linearVelocityScale=<0.6 0.2 0.2>
   }
   tactical_Sprint {
    angularVelocityScale=0.6
    fictitiousAngularScale=0.1
    jointName=root
    dragCoefficient=<2 2>
    gravityOverride=<0 0 -980>
    liftCoefficient=<0.5 0.5>
    linearVelocityScale=<0.6 0.6 0.5>
   }
   ziplining {
    angularVelocityScale=0.6
    fictitiousAngularScale=0
    jointName=root
    dragCoefficient=<0 1.5>
    gravityOverride=<0 0 -980>
    liftCoefficient=<0 0.3>
    linearVelocityScale=<0.4 0.4 0.5>
   }
  }
 }
 Cloth Paramaters_0 {
  type=ClothParamsStruct
  default {
   dBNO {
    angularVelocityScale=0.6
    fictitiousAngularScale=0.1
    jointName=pelvis
    dragCoefficient=<0.2 1>
    gravityOverride=<-980 400 0>
    liftCoefficient=<0 0.5>
    linearVelocityScale=<0.3 0.3 0.1>
   }
   dBNO_Carried {
    angularVelocityScale=0.6
    fictitiousAngularScale=0.1
    jointName=pelvis
    dragCoefficient=<0.2 1>
    gravityOverride=<-980 -200 0>
    liftCoefficient=<0 0.5>
    linearVelocityScale=<0.3 0.3 0.1>
   }
   dBNO_Impulse {
    angularVelocityScale=0
    fictitiousAngularScale=0
    jointName=pelvis
    dragCoefficient=<0 0.2>
    gravityOverride=<-3000 -2000 0>
    liftCoefficient=<0 0.1>
    linearVelocityScale=<0.3 0.1 0.3>
   }
   dBNO_RollBwd {
    angularVelocityScale=0
    fictitiousAngularScale=0
    jointName=pelvis
    dragCoefficient=<0 0.2>
    gravityOverride=<-2000 0 0>
    liftCoefficient=<0 0.1>
    linearVelocityScale=<0.3 0.1 0.3>
   }
   dBNO_RollFwd {
    angularVelocityScale=0
    fictitiousAngularScale=0
    jointName=pelvis
    dragCoefficient=<0 0.2>
    gravityOverride=<-980 0 0>
    liftCoefficient=<0 0.1>
    linearVelocityScale=<0.3 0.1 0.3>
   }
   dBNO_Tumbling {
    angularVelocityScale=0
    fictitiousAngularScale=0
    jointName=pelvis
    dragCoefficient=<0 0.2>
    gravityOverride=<-2500 -300 0>
    liftCoefficient=<0 0.1>
    linearVelocityScale=<0.3 0.1 0.3>
   }
   emoteOrMelee {
    angularVelocityScale=0.6
    fictitiousAngularScale=0.1
    jointName=root
    dragCoefficient=<2.5 2.5>
    gravityOverride=<0 0 -980>
    liftCoefficient=<0.5 0.5>
    linearVelocityScale=<0.6 0.6 0.5>
   }
   front_End {
    angularVelocityScale=0.05
    fictitiousAngularScale=0.6
    jointName=root
    dragCoefficient=<0.2 0.2>
    gravityOverride=<0 -200 -980>
    liftCoefficient=<0.1 0.1>
    linearVelocityScale=<0.1 0.1 0.1>
   }
   onGround_Crouch_Moving {
    angularVelocityScale=0.1
    fictitiousAngularScale=0.3
    jointName=root
    dragCoefficient=<2 2>
    gravityOverride=<0 0 -980>
    liftCoefficient=<0.5 0.5>
    linearVelocityScale=<0.6 0.6 0.5>
   }
   onGround_Falling {
    angularVelocityScale=0.6
    fictitiousAngularScale=0
    jointName=root
    dragCoefficient=<0 0.2>
    gravityOverride=<0 -500 -980>
    liftCoefficient=<0 0.1>
    linearVelocityScale=<2 1 0.3>
   }
   onGround_Moving {
    angularVelocityScale=0.6
    fictitiousAngularScale=0.1
    jointName=root
    dragCoefficient=<2 2>
    gravityOverride=<0 0 -980>
    liftCoefficient=<0.5 0.5>
    linearVelocityScale=<0.6 0.6 0.5>
   }
   onGround_Standing {
    angularVelocityScale=0.05
    fictitiousAngularScale=1.2
    jointName=root
    dragCoefficient=<2.5 2.5>
    gravityOverride=<0 0 -980>
    liftCoefficient=<0.5 0.5>
    linearVelocityScale=<0.6 0.6 0.5>
   }
   rail_Grinding {
    angularVelocityScale=0.2
    fictitiousAngularScale=0
    jointName=root
    dragCoefficient=<0.1 0.2>
    gravityOverride=<50 -500 -400>
    liftCoefficient=<0 0.5>
    linearVelocityScale=<0.1 0.1 0.1>
   }
   riding_Moving {
    angularVelocityScale=0.05
    fictitiousAngularScale=1
    jointName=root
    dragCoefficient=<0.5 0.5>
    gravityOverride=<0 0 -980>
    liftCoefficient=<0.5 0.5>
    linearVelocityScale=<0.4 0.6 0.3>
   }
   riding_Moving_Targeting {
    angularVelocityScale=0.05
    fictitiousAngularScale=0.7
    jointName=root
    dragCoefficient=<0.5 0.5>
    gravityOverride=<0 0 -980>
    liftCoefficient=<0.5 0.5>
    linearVelocityScale=<0.2 0.6 0.2>
   }
   riding_Standing {
    angularVelocityScale=0.05
    fictitiousAngularScale=1
    jointName=root
    dragCoefficient=<0.5 0.5>
    gravityOverride=<0 -450 -980>
    liftCoefficient=<0.5 0.5>
    linearVelocityScale=<0.6 0.6 0.3>
   }
   skydive_DivingDown {
    angularVelocityScale=0.1
    fictitiousAngularScale=0
    jointName=pelvis
    dragCoefficient=<0.3 0.3>
    gravityOverride=<-450 -100 0>
    liftCoefficient=<0.9 0.9>
    linearVelocityScale=<0.1 0.1 0.1>
   }
   skydive_FloatingDown {
    angularVelocityScale=0.1
    fictitiousAngularScale=0
    jointName=root
    dragCoefficient=<0.3 0.3>
    gravityOverride=<0 -800 200>
    liftCoefficient=<1.3 1.3>
    linearVelocityScale=<0.2 0.2 0.1>
   }
   skydive_GoingUp {
    angularVelocityScale=0.1
    fictitiousAngularScale=0
    jointName=root
    dragCoefficient=<0.01 0.01>
    gravityOverride=<0 -2000 400>
    liftCoefficient=<0.05 0.05>
    linearVelocityScale=<0.1 0.1 0.01>
   }
   skydive_Parachuting {
    angularVelocityScale=0.6
    fictitiousAngularScale=0.1
    jointName=root
    dragCoefficient=<0.8 0.8>
    gravityOverride=<0 -600 -500>
    liftCoefficient=<0.8 0.8>
    linearVelocityScale=<0.2 0.2 0.3>
   }
   skydive_Surfing {
    angularVelocityScale=0.6
    fictitiousAngularScale=0.1
    jointName=root
    dragCoefficient=<1.5 1.5>
    gravityOverride=<0 -600 -500>
    liftCoefficient=<1 1>
    linearVelocityScale=<0.1 0.1 0.3>
   }
   swimming {
    angularVelocityScale=0.6
    fictitiousAngularScale=0
    jointName=root
    dragCoefficient=<2.5 2.5>
    gravityOverride=<0 0 0>
    liftCoefficient=<0.1 0.1>
    linearVelocityScale=<0.6 0.6 0.5>
   }
   tactical_Sprint {
    angularVelocityScale=0.6
    fictitiousAngularScale=0.1
    jointName=root
    dragCoefficient=<2 2>
    gravityOverride=<0 0 -980>
    liftCoefficient=<0.5 0.5>
    linearVelocityScale=<0.6 0.6 0.5>
   }
   ziplining {
    angularVelocityScale=0.6
    fictitiousAngularScale=0
    jointName=root
    dragCoefficient=<0 1.5>
    gravityOverride=<0 0 -980>
    liftCoefficient=<0 0.3>
    linearVelocityScale=<0.4 0.4 0.5>
   }
  }
 }
 consumable {type=boolean}
 DeltaTimeLocal {type=number}
 Gravity Override Params {type=GravityOverrideParamsStruct, default {skydive_FloatingDown {gravityOverride=<0 -700 900>}}}
 Gravity Override Params_cape {
  type=GravityOverrideParamsStruct
  default {onGround_Standing {gravityOverride=<0 0 -230>, windAmplitude=<50 -50 10>, windFrequency=<0.15 0.8 0.5>, windOffset=<0 -80 -15>}}
 }
 GravityOverride {type=Vector}
 GravityOverride_cape {type=Vector}
 Gust Local Settings {type=WindGustLocalSettingStruct, default {elapsedTime=0, finalGustStrengthLocal=0, gustDuration=0, gustStrength=0, timeBetweenGusts=0}}
 head rot left {type=number}
 head rot right {type=number}
 head tilt right {type=number}
 Is Skydive Floating {type=boolean}
 is Targeting float {type=number}
 Is Wearing Backbling {type=boolean}
 isCrouchingFloat {type=number}
 isIdleFrontEnd {type=boolean}
 isNOTCrouchingFloat {type=number}
 look Up ADS {type=number}
 LookDown {type=number}
 LookUp {type=number}
 LookUpRifle {type=number}
 melee {type=boolean}
 Pawn {type=FortPlayerPawn}
 pistol {type=boolean}
 PlummetAlpha {type=number}
 rifle {type=boolean}
 Skel_Mesh {type=SkeletalMeshComponent}
 Thigh L Rot Neg {type=number}
 Thigh L Rot Pos {type=number}
 TrailControllerRot {type=Rotator}
 Wind Gusts Params {
  type=WindGustParamsStruct
  default {
   dBNO {
    maxDownInterpSpeed=5
    maxGustDuration=2
    maxGustStrength=7
    maxTimeBetweenGusts=7
    maxUpInterpSpeed=5
    minDownInterpSpeed=2
    minGustDuration=0.5
    minGustStrength=3
    minTimeBetweenGusts=0
    minUpInterpSpeed=2
    windTurbulenceScalarA=0.1
    windTurbulenceScalarB=1
    windAmplitude=<0 0 0>
    windFrequency=<0 0 0>
    windOffset=<0 0 0>
   }
   dBNO_Carried {
    maxDownInterpSpeed=5
    maxGustDuration=2
    maxGustStrength=7
    maxTimeBetweenGusts=7
    maxUpInterpSpeed=5
    minDownInterpSpeed=2
    minGustDuration=0.5
    minGustStrength=3
    minTimeBetweenGusts=0
    minUpInterpSpeed=2
    windTurbulenceScalarA=0.1
    windTurbulenceScalarB=1
    windAmplitude=<0 0 0>
    windFrequency=<0 0 0>
    windOffset=<0 0 0>
   }
   dBNO_Impulse {
    maxDownInterpSpeed=5
    maxGustDuration=2
    maxGustStrength=7
    maxTimeBetweenGusts=7
    maxUpInterpSpeed=5
    minDownInterpSpeed=2
    minGustDuration=0.5
    minGustStrength=3
    minTimeBetweenGusts=0
    minUpInterpSpeed=2
    windTurbulenceScalarA=0.1
    windTurbulenceScalarB=1
    windAmplitude=<0 0 0>
    windFrequency=<0 0 0>
    windOffset=<0 0 0>
   }
   dBNO_RollBwd {
    maxDownInterpSpeed=5
    maxGustDuration=2
    maxGustStrength=7
    maxTimeBetweenGusts=7
    maxUpInterpSpeed=5
    minDownInterpSpeed=2
    minGustDuration=0.5
    minGustStrength=3
    minTimeBetweenGusts=0
    minUpInterpSpeed=2
    windTurbulenceScalarA=0.1
    windTurbulenceScalarB=1
    windAmplitude=<0 0 0>
    windFrequency=<0 0 0>
    windOffset=<0 0 0>
   }
   dBNO_RollFwd {
    maxDownInterpSpeed=5
    maxGustDuration=2
    maxGustStrength=7
    maxTimeBetweenGusts=7
    maxUpInterpSpeed=5
    minDownInterpSpeed=2
    minGustDuration=0.5
    minGustStrength=3
    minTimeBetweenGusts=0
    minUpInterpSpeed=2
    windTurbulenceScalarA=0.1
    windTurbulenceScalarB=1
    windAmplitude=<0 0 0>
    windFrequency=<0 0 0>
    windOffset=<0 0 0>
   }
   dBNO_Tumbling {
    maxDownInterpSpeed=5
    maxGustDuration=2
    maxGustStrength=7
    maxTimeBetweenGusts=7
    maxUpInterpSpeed=5
    minDownInterpSpeed=2
    minGustDuration=0.5
    minGustStrength=3
    minTimeBetweenGusts=0
    minUpInterpSpeed=2
    windTurbulenceScalarA=0.1
    windTurbulenceScalarB=1
    windAmplitude=<0 0 0>
    windFrequency=<0 0 0>
    windOffset=<0 0 0>
   }
   emoteOrMelee {
    maxDownInterpSpeed=5
    maxGustDuration=2
    maxGustStrength=7
    maxTimeBetweenGusts=7
    maxUpInterpSpeed=5
    minDownInterpSpeed=2
    minGustDuration=0.5
    minGustStrength=3
    minTimeBetweenGusts=0
    minUpInterpSpeed=2
    windTurbulenceScalarA=0.1
    windTurbulenceScalarB=1
    windAmplitude=<30 25 50>
    windFrequency=<2 3 5>
    windOffset=<-300 -150 0>
   }
   front_End {
    maxDownInterpSpeed=5
    maxGustDuration=2
    maxGustStrength=4
    maxTimeBetweenGusts=7
    maxUpInterpSpeed=5
    minDownInterpSpeed=2
    minGustDuration=0.5
    minGustStrength=2
    minTimeBetweenGusts=0
    minUpInterpSpeed=2
    windTurbulenceScalarA=0.1
    windTurbulenceScalarB=1
    windAmplitude=<10 30 50>
    windFrequency=<10 3 4>
    windOffset=<0 -200 -250>
   }
   onGround_Crouch_Moving {
    maxDownInterpSpeed=5
    maxGustDuration=2
    maxGustStrength=7
    maxTimeBetweenGusts=7
    maxUpInterpSpeed=5
    minDownInterpSpeed=2
    minGustDuration=0.5
    minGustStrength=3
    minTimeBetweenGusts=0
    minUpInterpSpeed=2
    windTurbulenceScalarA=0.1
    windTurbulenceScalarB=1
    windAmplitude=<100 25 75>
    windFrequency=<1.5 0.5 7.5>
    windOffset=<0 -100 -175>
   }
   onGround_Falling {
    maxDownInterpSpeed=5
    maxGustDuration=2
    maxGustStrength=7
    maxTimeBetweenGusts=7
    maxUpInterpSpeed=5
    minDownInterpSpeed=2
    minGustDuration=0.5
    minGustStrength=3
    minTimeBetweenGusts=0
    minUpInterpSpeed=2
    windTurbulenceScalarA=0.1
    windTurbulenceScalarB=1
    windAmplitude=<0 0 0>
    windFrequency=<0 0 0>
    windOffset=<0 0 0>
   }
   onGround_Moving {
    maxDownInterpSpeed=5
    maxGustDuration=8
    maxGustStrength=7
    maxTimeBetweenGusts=8
    maxUpInterpSpeed=5
    minDownInterpSpeed=2
    minGustDuration=0.2
    minGustStrength=3
    minTimeBetweenGusts=1
    minUpInterpSpeed=2
    windTurbulenceScalarA=0.1
    windTurbulenceScalarB=0.8
    windAmplitude=<30 65 42>
    windFrequency=<5 6 8.6>
    windOffset=<30 -250 -300>
   }
   onGround_Standing {
    maxDownInterpSpeed=5
    maxGustDuration=8
    maxGustStrength=7
    maxTimeBetweenGusts=8
    maxUpInterpSpeed=5
    minDownInterpSpeed=2
    minGustDuration=0.2
    minGustStrength=2
    minTimeBetweenGusts=1
    minUpInterpSpeed=2
    windTurbulenceScalarA=0.1
    windTurbulenceScalarB=0.8
    windAmplitude=<75 24 28>
    windFrequency=<1.5 2.3 3.7>
    windOffset=<75 -100 -150>
   }
   rail_Grinding {
    maxDownInterpSpeed=5
    maxGustDuration=8
    maxGustStrength=7
    maxTimeBetweenGusts=8
    maxUpInterpSpeed=5
    minDownInterpSpeed=2
    minGustDuration=0.2
    minGustStrength=3
    minTimeBetweenGusts=1
    minUpInterpSpeed=2
    windTurbulenceScalarA=0.1
    windTurbulenceScalarB=0.8
    windAmplitude=<20 65 42>
    windFrequency=<5 6 8.6>
    windOffset=<50 -500 -400>
   }
   riding_Moving {
    maxDownInterpSpeed=5
    maxGustDuration=2
    maxGustStrength=7
    maxTimeBetweenGusts=7
    maxUpInterpSpeed=5
    minDownInterpSpeed=2
    minGustDuration=0.5
    minGustStrength=3
    minTimeBetweenGusts=0
    minUpInterpSpeed=2
    windTurbulenceScalarA=0.1
    windTurbulenceScalarB=1
    windAmplitude=<0 0 0>
    windFrequency=<0 0 0>
    windOffset=<0 0 0>
   }
   riding_Moving_Targeting {
    maxDownInterpSpeed=5
    maxGustDuration=2
    maxGustStrength=7
    maxTimeBetweenGusts=7
    maxUpInterpSpeed=5
    minDownInterpSpeed=2
    minGustDuration=0.5
    minGustStrength=3
    minTimeBetweenGusts=0
    minUpInterpSpeed=2
    windTurbulenceScalarA=0.1
    windTurbulenceScalarB=1
    windAmplitude=<0 0 0>
    windFrequency=<0 0 0>
    windOffset=<0 0 0>
   }
   riding_Standing {
    maxDownInterpSpeed=5
    maxGustDuration=2
    maxGustStrength=7
    maxTimeBetweenGusts=7
    maxUpInterpSpeed=5
    minDownInterpSpeed=2
    minGustDuration=0.5
    minGustStrength=3
    minTimeBetweenGusts=0
    minUpInterpSpeed=2
    windTurbulenceScalarA=0.1
    windTurbulenceScalarB=1
    windAmplitude=<0 0 0>
    windFrequency=<0 0 0>
    windOffset=<0 0 0>
   }
   skydive_DivingDown {
    maxDownInterpSpeed=5
    maxGustDuration=8
    maxGustStrength=7
    maxTimeBetweenGusts=8
    maxUpInterpSpeed=5
    minDownInterpSpeed=2
    minGustDuration=0.2
    minGustStrength=3
    minTimeBetweenGusts=1
    minUpInterpSpeed=2
    windTurbulenceScalarA=0.1
    windTurbulenceScalarB=4
    windAmplitude=<0 25 35>
    windFrequency=<0 4 10>
    windOffset=<-450 -100 100>
   }
   skydive_FloatingDown {
    maxDownInterpSpeed=5
    maxGustDuration=8
    maxGustStrength=1
    maxTimeBetweenGusts=8
    maxUpInterpSpeed=5
    minDownInterpSpeed=2
    minGustDuration=0.2
    minGustStrength=0
    minTimeBetweenGusts=1
    minUpInterpSpeed=2
    windTurbulenceScalarA=0.5
    windTurbulenceScalarB=1
    windAmplitude=<50 20 35>
    windFrequency=<8 2 3>
    windOffset=<0 -300 200>
   }
   skydive_GoingUp {
    maxDownInterpSpeed=5
    maxGustDuration=8
    maxGustStrength=7
    maxTimeBetweenGusts=8
    maxUpInterpSpeed=5
    minDownInterpSpeed=2
    minGustDuration=0.2
    minGustStrength=3
    minTimeBetweenGusts=1
    minUpInterpSpeed=2
    windTurbulenceScalarA=0.1
    windTurbulenceScalarB=0.8
    windAmplitude=<0 0 0>
    windFrequency=<0 0 0>
    windOffset=<0 -500 300>
   }
   skydive_Parachuting {
    maxDownInterpSpeed=5
    maxGustDuration=8
    maxGustStrength=1
    maxTimeBetweenGusts=8
    maxUpInterpSpeed=5
    minDownInterpSpeed=2
    minGustDuration=5
    minGustStrength=0
    minTimeBetweenGusts=5
    minUpInterpSpeed=2
    windTurbulenceScalarA=0.2
    windTurbulenceScalarB=1
    windAmplitude=<25 50 75>
    windFrequency=<8 6 2>
    windOffset=<-150 -400 -150>
   }
   skydive_Surfing {
    maxDownInterpSpeed=5
    maxGustDuration=2
    maxGustStrength=7
    maxTimeBetweenGusts=7
    maxUpInterpSpeed=5
    minDownInterpSpeed=2
    minGustDuration=0.5
    minGustStrength=3
    minTimeBetweenGusts=0
    minUpInterpSpeed=2
    windTurbulenceScalarA=0.1
    windTurbulenceScalarB=1
    windAmplitude=<0 0 0>
    windFrequency=<0 0 0>
    windOffset=<0 0 0>
   }
   swimming {
    maxDownInterpSpeed=5
    maxGustDuration=2
    maxGustStrength=7
    maxTimeBetweenGusts=7
    maxUpInterpSpeed=5
    minDownInterpSpeed=2
    minGustDuration=0.5
    minGustStrength=3
    minTimeBetweenGusts=0
    minUpInterpSpeed=2
    windTurbulenceScalarA=0.1
    windTurbulenceScalarB=1
    windAmplitude=<10 25 20>
    windFrequency=<1.5 0.5 7.5>
    windOffset=<0 -150 50>
   }
   tactical_Sprint {
    maxDownInterpSpeed=5
    maxGustDuration=2
    maxGustStrength=7
    maxTimeBetweenGusts=7
    maxUpInterpSpeed=5
    minDownInterpSpeed=2
    minGustDuration=0.5
    minGustStrength=3
    minTimeBetweenGusts=0
    minUpInterpSpeed=2
    windTurbulenceScalarA=0.1
    windTurbulenceScalarB=1
    windAmplitude=<0 0 0>
    windFrequency=<0 0 0>
    windOffset=<0 0 0>
   }
   ziplining {
    maxDownInterpSpeed=5
    maxGustDuration=2
    maxGustStrength=7
    maxTimeBetweenGusts=7
    maxUpInterpSpeed=5
    minDownInterpSpeed=2
    minGustDuration=0.5
    minGustStrength=3
    minTimeBetweenGusts=0
    minUpInterpSpeed=2
    windTurbulenceScalarA=0.1
    windTurbulenceScalarB=1
    windAmplitude=<0 0 0>
    windFrequency=<0 0 0>
    windOffset=<0 0 0>
   }
  }
 }
]
